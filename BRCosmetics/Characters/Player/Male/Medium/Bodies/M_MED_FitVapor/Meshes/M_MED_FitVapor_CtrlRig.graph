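# Control Rig for the M_MED_FitVapor player body that drives secondary and corrective bones for hood, collar, tubes, tailcoat, radio and limb deformation.
# Weighted parent constraints attach garment and accessory bones to body bones.
# A shared DeltaRotation function remaps swing or twist angles to weights; these, curves and spherical pose readers blend corrective transform offsets.
# A debug-gated branch copies controls onto bones and computes limb twist.
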ControlRigBlueprint /BRCosmetics/Characters/Player/Male/Medium/Bodies/M_MED_FitVapor/Meshes/M_MED_FitVapor_CtrlRig
graphs [
 CollapseNode_ContainedGraph {nodes [
   RigVMFunctionEntryNode()
   > &RigUnit_ParentConstraint(
    AdvancedSettings=(RotationOrderForFilter=XZY)
    Child=(Type=Bone,Name="base_hood_c")
    Filter=(TranslationFilter=(bX=True,bY=True,bZ=True),RotationFilter=(bX=False,bY=False,bZ=False),ScaleFilter=(bX=False,bY=False,bZ=False))
    Parents="((Item=(Type=Bone,Name=\"upperarm_twist_01_r\"),Weight=0.0001),(Item=(Type=Bone,Name=\"upperarm_twist_01_l\"),Weight=0.0004),(Item=(Type=Bone,Name=\"clavicle_r\"),Weight=0.0133),(Item=(Type=Bone,Name=\"spine_04\"),Weight=0.0151),(Item=(Type=Bone,Name=\"spine_03\"),W ...(truncated)"
    Weight=1
   )
   > &RigUnit_ParentConstraint(
    AdvancedSettings=(RotationOrderForFilter=XZY)
    Child=(Type=Bone,Name="base_hood_3_l")
    Filter=(TranslationFilter=(bX=True,bY=True,bZ=True),RotationFilter=(bX=False,bY=False,bZ=False),ScaleFilter=(bX=False,bY=False,bZ=False))
    Parents=((Item=(Type=Bone,Name="clavicle_l"),Weight=0.2997),(Item=(Type=Bone,Name="spine_05"),Weight=0.7003))
    Weight=1
   )
   > &RigUnit_ParentConstraint(
    AdvancedSettings=(RotationOrderForFilter=XZY)
    Child=(Type=Bone,Name="base_hood_2_l")
    Filter=(TranslationFilter=(bX=True,bY=True,bZ=True),RotationFilter=(bX=False,bY=False,bZ=False),ScaleFilter=(bX=False,bY=False,bZ=False))
    Parents="((Item=(Type=Bone,Name=\"deform_pec_l\"),Weight=0.0019),(Item=(Type=Bone,Name=\"spine_04\"),Weight=0.0048),(Item=(Type=Bone,Name=\"deform_upperarm_fr_l\"),Weight=0.005),(Item=(Type=Bone,Name=\"spine_03\"),Weight=0.0056),(Item=(Type=Bone,Name=\"deform_upperarm_up_l\" ...(truncated)"
    Weight=1
   )
   > &RigUnit_ParentConstraint(
    AdvancedSettings=(RotationOrderForFilter=XZY)
    Child=(Type=Bone,Name="base_hood_1_l")
    Filter=(TranslationFilter=(bX=True,bY=True,bZ=True),RotationFilter=(bX=False,bY=False,bZ=False),ScaleFilter=(bX=False,bY=False,bZ=False))
    Parents="((Item=(Type=Bone,Name=\"deform_upperarm_up_l\"),Weight=0.0081),(Item=(Type=Bone,Name=\"deform_upperarm_fr_l\"),Weight=0.0099),(Item=(Type=Bone,Name=\"spine_03\"),Weight=0.0206),(Item=(Type=Bone,Name=\"spine_04\"),Weight=0.0524),(Item=(Type=Bone,Name=\"deform_pec_l ...(truncated)"
    Weight=1
   )
   > &RigUnit_ParentConstraint(
    AdvancedSettings=(RotationOrderForFilter=XZY)
    Child=(Type=Bone,Name="base_hood_3_r")
    Filter=(TranslationFilter=(bX=True,bY=True,bZ=True),RotationFilter=(bX=False,bY=False,bZ=False),ScaleFilter=(bX=False,bY=False,bZ=False))
    Parents=((Item=(Type=Bone,Name="clavicle_r"),Weight=0.3717),(Item=(Type=Bone,Name="spine_05"),Weight=0.6283))
    Weight=1
   )
   > &RigUnit_ParentConstraint(
    AdvancedSettings=(RotationOrderForFilter=XZY)
    Child=(Type=Bone,Name="base_hood_2_r")
    Filter=(TranslationFilter=(bX=True,bY=True,bZ=True),RotationFilter=(bX=False,bY=False,bZ=False),ScaleFilter=(bX=False,bY=False,bZ=False))
    Parents="((Item=(Type=Bone,Name=\"deform_pec_r\"),Weight=0.0021),(Item=(Type=Bone,Name=\"spine_04\"),Weight=0.005),(Item=(Type=Bone,Name=\"deform_upperarm_fr_r\"),Weight=0.0054),(Item=(Type=Bone,Name=\"spine_03\"),Weight=0.0059),(Item=(Type=Bone,Name=\"deform_upperarm_up_r\" ...(truncated)"
    Weight=1
   )
   > &RigUnit_ParentConstraint(
    AdvancedSettings=(RotationOrderForFilter=XZY)
    Child=(Type=Bone,Name="base_hood_1_r")
    Filter=(TranslationFilter=(bX=True,bY=True,bZ=True),RotationFilter=(bX=False,bY=False,bZ=False),ScaleFilter=(bX=False,bY=False,bZ=False))
    Parents="((Item=(Type=Bone,Name=\"deform_upperarm_up_r\"),Weight=0.0096),(Item=(Type=Bone,Name=\"deform_upperarm_fr_r\"),Weight=0.0117),(Item=(Type=Bone,Name=\"spine_03\"),Weight=0.0203),(Item=(Type=Bone,Name=\"spine_04\"),Weight=0.0509),(Item=(Type=Bone,Name=\"deform_pec_r ...(truncated)"
    Weight=1
   )
   > RigVMFunctionReturnNode()
  ]}
 CollapseNode_ContainedGraph {nodes [
   RigVMFunctionEntryNode()
   > &RigUnit_ParentConstraint(
    AdvancedSettings=(RotationOrderForFilter=XZY)
    Child=(Type=Bone,Name="base_internalNeckCollar_1_r")
    Filter=(TranslationFilter=(bX=True,bY=True,bZ=True),RotationFilter=(bX=false,bY=false,bZ=false),ScaleFilter=(bX=True,bY=True,bZ=True))
    Parents="((Item=(Type=Bone,Name=\"deform_upperarm_up_r\"),Weight=0.0041),(Item=(Type=Bone,Name=\"deform_upperarm_fr_r\"),Weight=0.0055),(Item=(Type=Bone,Name=\"neck_01\"),Weight=0.0171),(Item=(Type=Bone,Name=\"deform_lat_r\"),Weight=0.0193),(Item=(Type=Bone,Name=\"spine_04\" ...(truncated)"
    Weight=1
   )
   > &RigUnit_ParentConstraint(
    AdvancedSettings=(RotationOrderForFilter=XZY)
    Child=(Type=Bone,Name="base_internalNeckCollar_1_c")
    Filter=(TranslationFilter=(bX=True,bY=True,bZ=True),RotationFilter=(bX=false,bY=false,bZ=false),ScaleFilter=(bX=True,bY=True,bZ=True))
    Parents="((Item=(Type=Bone,Name=\"deform_pec_l\"),Weight=0.0002),(Item=(Type=Bone,Name=\"deform_upperarm_fr_l\"),Weight=0.0003),(Item=(Type=Bone,Name=\"deform_upperarm_fr_r\"),Weight=0.0003),(Item=(Type=Bone,Name=\"deform_lat_r\"),Weight=0.0011),(Item=(Type=Bone,Name=\"spin ...(truncated)"
    Weight=1
   )
   > &RigUnit_ParentConstraint(
    AdvancedSettings=(RotationOrderForFilter=XZY)
    Child=(Type=Bone,Name="base_internalNeckCollar_1_l")
    Filter=(TranslationFilter=(bX=True,bY=True,bZ=True),RotationFilter=(bX=false,bY=false,bZ=false),ScaleFilter=(bX=True,bY=True,bZ=True))
    Parents="((Item=(Type=Bone,Name=\"deform_upperarm_up_l\"),Weight=0.0041),(Item=(Type=Bone,Name=\"deform_upperarm_fr_l\"),Weight=0.0055),(Item=(Type=Bone,Name=\"neck_01\"),Weight=0.0184),(Item=(Type=Bone,Name=\"deform_lat_l\"),Weight=0.0192),(Item=(Type=Bone,Name=\"spine_04\" ...(truncated)"
    Weight=1
   )
   > RigVMFunctionReturnNode()
  ]}
 CollapseNode_ContainedGraph {nodes [
   RigVMFunctionEntryNode()
   > &RigUnit_ParentConstraint(
    AdvancedSettings=(RotationOrderForFilter=XZY)
    Child=(Type=Bone,Name="stk_sTube_a_2_r")
    Parents="((Item=(Type=Bone,Name=\"deform_upperarm_bk_r\"),Weight=0.0003),(Item=(Type=Bone,Name=\"spine_04\"),Weight=0.0003),(Item=(Type=Bone,Name=\"deform_lat_r\"),Weight=0.0085),(Item=(Type=Bone,Name=\"upperarm_twist_01_r\"),Weight=0.0135),(Item=(Type=Bone,Name=\"deform_up ...(truncated)"
    Weight=1
   )
   > &RigUnit_ParentConstraint(
    AdvancedSettings=(RotationOrderForFilter=XZY)
    Child=(Type=Bone,Name="stk_sTube_a_3_r")
    Parents="((Item=(Type=Bone,Name=\"spine_03\"),Weight=0.0003),(Item=(Type=Bone,Name=\"spine_04\"),Weight=0.0038),(Item=(Type=Bone,Name=\"deform_lat_r\"),Weight=0.019),(Item=(Type=Bone,Name=\"upperarm_twist_01_r\"),Weight=0.0296),(Item=(Type=Bone,Name=\"deform_upperarm_bk_r\") ...(truncated)"
    Weight=1
   )
   > &RigUnit_ParentConstraint(
    AdvancedSettings=(RotationOrderForFilter=XZY)
    Child=(Type=Bone,Name="stk_sTube_a_1_r")
    Parents="((Item=(Type=Bone,Name=\"upperarm_twist_01_r\"),Weight=0.0181),(Item=(Type=Bone,Name=\"spine_04\"),Weight=0.0234),(Item=(Type=Bone,Name=\"deform_upperarm_up_r\"),Weight=0.0589),(Item=(Type=Bone,Name=\"deform_pec_r\"),Weight=0.0759),(Item=(Type=Bone,Name=\"deform_up ...(truncated)"
    Weight=1
   )
   > &RigUnit_ParentConstraint(
    AdvancedSettings=(RotationOrderForFilter=XZY)
    Child=(Type=Bone,Name="stk_sTube_b_3_r")
    Parents="((Item=(Type=Bone,Name=\"upperarm_twist_02_r\"),Weight=0.0004),(Item=(Type=Bone,Name=\"spine_04\"),Weight=0.0028),(Item=(Type=Bone,Name=\"deform_upperarm_up_r\"),Weight=0.0548),(Item=(Type=Bone,Name=\"deform_lat_r\"),Weight=0.0553),(Item=(Type=Bone,Name=\"spine_05\" ...(truncated)"
    Weight=1
   )
   > &RigUnit_ParentConstraint(
    AdvancedSettings=(RotationOrderForFilter=XZY)
    Child=(Type=Bone,Name="stk_sTube_b_2_r")
    Parents="((Item=(Type=Bone,Name=\"upperarm_twist_02_r\"),Weight=0.0007),(Item=(Type=Bone,Name=\"deform_upperarm_bk_r\"),Weight=0.0027),(Item=(Type=Bone,Name=\"deform_lat_r\"),Weight=0.0155),(Item=(Type=Bone,Name=\"spine_05\"),Weight=0.0498),(Item=(Type=Bone,Name=\"deform_up ...(truncated)"
    Weight=1
   )
   > &RigUnit_ParentConstraint(
    AdvancedSettings=(RotationOrderForFilter=XZY)
    Child=(Type=Bone,Name="stk_sTube_b_1_r")
    Parents="((Item=(Type=Bone,Name=\"spine_04\"),Weight=0.0046),(Item=(Type=Bone,Name=\"deform_pec_r\"),Weight=0.0281),(Item=(Type=Bone,Name=\"deform_upperarm_up_r\"),Weight=0.0313),(Item=(Type=Bone,Name=\"spine_05\"),Weight=0.0572),(Item=(Type=Bone,Name=\"upperarm_twist_01_r\" ...(truncated)"
    Weight=1
   )
   > &RigUnit_ParentConstraint(
    AdvancedSettings=(RotationOrderForFilter=XZY)
    Child=(Type=Bone,Name="stk_sTube_c_3_r")
    Parents="((Item=(Type=Bone,Name=\"upperarm_r\"),Weight=0.0002),(Item=(Type=Bone,Name=\"upperarm_twist_02_r\"),Weight=0.0074),(Item=(Type=Bone,Name=\"spine_05\"),Weight=0.0175),(Item=(Type=Bone,Name=\"deform_lat_r\"),Weight=0.033),(Item=(Type=Bone,Name=\"deform_upperarm_up_r ...(truncated)"
    Weight=1
   )
   > &RigUnit_ParentConstraint(
    AdvancedSettings=(RotationOrderForFilter=XZY)
    Child=(Type=Bone,Name="stk_sTube_c_2_r")
    Parents="((Item=(Type=Bone,Name=\"deform_upperarm_bk_r\"),Weight=0.0017),(Item=(Type=Bone,Name=\"spine_05\"),Weight=0.0127),(Item=(Type=Bone,Name=\"upperarm_twist_02_r\"),Weight=0.0178),(Item=(Type=Bone,Name=\"deform_lat_r\"),Weight=0.02),(Item=(Type=Bone,Name=\"deform_uppe ...(truncated)"
    Weight=1
   )
   > &RigUnit_ParentConstraint(
    AdvancedSettings=(RotationOrderForFilter=XZY)
    Child=(Type=Bone,Name="stk_sTube_c_1_r")
    Parents="((Item=(Type=Bone,Name=\"deform_pec_r\"),Weight=0.0067),(Item=(Type=Bone,Name=\"spine_05\"),Weight=0.0101),(Item=(Type=Bone,Name=\"upperarm_twist_02_r\"),Weight=0.0171),(Item=(Type=Bone,Name=\"deform_upperarm_up_r\"),Weight=0.0652),(Item=(Type=Bone,Name=\"clavicle_ ...(truncated)"
    Weight=1
   )
   > &RigUnit_ParentConstraint(
    AdvancedSettings=(RotationOrderForFilter=XZY)
    Child=(Type=Bone,Name="stk_sTube_c_1_l")
    Parents="((Item=(Type=Bone,Name=\"deform_pec_l\"),Weight=0.0067),(Item=(Type=Bone,Name=\"spine_05\"),Weight=0.0101),(Item=(Type=Bone,Name=\"upperarm_twist_02_l\"),Weight=0.0171),(Item=(Type=Bone,Name=\"deform_upperarm_up_l\"),Weight=0.0652),(Item=(Type=Bone,Name=\"clavicle_ ...(truncated)"
    Weight=1
   )
   > &RigUnit_ParentConstraint(
    AdvancedSettings=(RotationOrderForFilter=XZY)
    Child=(Type=Bone,Name="stk_sTube_c_2_l")
    Parents="((Item=(Type=Bone,Name=\"deform_upperarm_bk_l\"),Weight=0.0017),(Item=(Type=Bone,Name=\"spine_05\"),Weight=0.0127),(Item=(Type=Bone,Name=\"upperarm_twist_02_l\"),Weight=0.0178),(Item=(Type=Bone,Name=\"deform_lat_l\"),Weight=0.02),(Item=(Type=Bone,Name=\"deform_uppe ...(truncated)"
    Weight=1
   )
   > &RigUnit_ParentConstraint(
    AdvancedSettings=(RotationOrderForFilter=XZY)
    Child=(Type=Bone,Name="stk_sTube_c_3_l")
    Parents="((Item=(Type=Bone,Name=\"upperarm_l\"),Weight=0.0002),(Item=(Type=Bone,Name=\"upperarm_twist_02_l\"),Weight=0.0074),(Item=(Type=Bone,Name=\"spine_05\"),Weight=0.0175),(Item=(Type=Bone,Name=\"deform_lat_l\"),Weight=0.033),(Item=(Type=Bone,Name=\"deform_upperarm_up_l ...(truncated)"
    Weight=1
   )
   > &RigUnit_ParentConstraint(
    AdvancedSettings=(RotationOrderForFilter=XZY)
    Child=(Type=Bone,Name="stk_sTube_b_1_l")
    Parents="((Item=(Type=Bone,Name=\"spine_04\"),Weight=0.0046),(Item=(Type=Bone,Name=\"deform_pec_l\"),Weight=0.0281),(Item=(Type=Bone,Name=\"deform_upperarm_up_l\"),Weight=0.0313),(Item=(Type=Bone,Name=\"spine_05\"),Weight=0.0572),(Item=(Type=Bone,Name=\"upperarm_twist_01_l\" ...(truncated)"
    Weight=1
   )
   > &RigUnit_ParentConstraint(
    AdvancedSettings=(RotationOrderForFilter=XZY)
    Child=(Type=Bone,Name="stk_sTube_b_2_l")
    Parents="((Item=(Type=Bone,Name=\"upperarm_twist_02_l\"),Weight=0.0007),(Item=(Type=Bone,Name=\"deform_upperarm_bk_l\"),Weight=0.0027),(Item=(Type=Bone,Name=\"deform_lat_l\"),Weight=0.0155),(Item=(Type=Bone,Name=\"spine_05\"),Weight=0.0498),(Item=(Type=Bone,Name=\"deform_up ...(truncated)"
    Weight=1
   )
   > &RigUnit_ParentConstraint(
    AdvancedSettings=(RotationOrderForFilter=XZY)
    Child=(Type=Bone,Name="stk_sTube_b_3_l")
    Parents="((Item=(Type=Bone,Name=\"upperarm_twist_02_l\"),Weight=0.0004),(Item=(Type=Bone,Name=\"spine_04\"),Weight=0.0028),(Item=(Type=Bone,Name=\"deform_upperarm_up_l\"),Weight=0.0548),(Item=(Type=Bone,Name=\"deform_lat_l\"),Weight=0.0553),(Item=(Type=Bone,Name=\"spine_05\" ...(truncated)"
    Weight=1
   )
   > &RigUnit_ParentConstraint(
    AdvancedSettings=(RotationOrderForFilter=XZY)
    Child=(Type=Bone,Name="stk_sTube_a_1_l")
    Parents="((Item=(Type=Bone,Name=\"upperarm_twist_01_l\"),Weight=0.0181),(Item=(Type=Bone,Name=\"spine_04\"),Weight=0.0234),(Item=(Type=Bone,Name=\"deform_upperarm_up_l\"),Weight=0.0589),(Item=(Type=Bone,Name=\"deform_pec_l\"),Weight=0.0759),(Item=(Type=Bone,Name=\"deform_up ...(truncated)"
    Weight=1
   )
   > &RigUnit_ParentConstraint(
    AdvancedSettings=(RotationOrderForFilter=XZY)
    Child=(Type=Bone,Name="stk_sTube_a_3_l")
    Parents="((Item=(Type=Bone,Name=\"spine_03\"),Weight=0.0003),(Item=(Type=Bone,Name=\"spine_04\"),Weight=0.0038),(Item=(Type=Bone,Name=\"deform_lat_l\"),Weight=0.019),(Item=(Type=Bone,Name=\"upperarm_twist_01_l\"),Weight=0.0296),(Item=(Type=Bone,Name=\"deform_upperarm_bk_l\") ...(truncated)"
    Weight=1
   )
   > &RigUnit_ParentConstraint(
    AdvancedSettings=(RotationOrderForFilter=XZY)
    Child=(Type=Bone,Name="stk_sTube_a_2_l")
    Parents="((Item=(Type=Bone,Name=\"deform_upperarm_bk_l\"),Weight=0.0003),(Item=(Type=Bone,Name=\"spine_04\"),Weight=0.0003),(Item=(Type=Bone,Name=\"deform_lat_l\"),Weight=0.0085),(Item=(Type=Bone,Name=\"upperarm_twist_01_l\"),Weight=0.0135),(Item=(Type=Bone,Name=\"deform_up ...(truncated)"
    Weight=1
   )
   > RigVMFunctionReturnNode()
  ]}
 New Function_ContainedGraph {nodes [
   RigVMFunctionEntryNode()
   > DeltaRotation = RigVMFunctionReferenceNode(Rotation Order=ZYX, child=hand_l, max=100, min=70, onlyTwist=true, parent=hand_l_null, parentType=Null, premult=1, twistAxis=(X=1,Y=0,Z=0))
   > &RigUnit_ModifyTransforms(
    ItemToModify="((Item=(Type=Bone,Name=\"lowerarm_twist_01_l\"),Transform=(Rotation=(X=0,Y=-0.0262,Z=0,W=0.9997),Translation=(X=-0.7,Y=-0.17,Z=1),Scale3D=(X=1,Y=1,Z=1))),(Item=(Type=Bone,Name=\"lowerarm_twist_02_l\"),Transform=(Rotation=(X=0.342,Y=0,Z=0,W=0.9397),Translation= ...(truncated)"
    WeightMinimum=0
    Weight=@DeltaRotation.outputWeight
   )
   > RigVMFunctionReturnNode()
  ]}
 New Function_ContainedGraph {nodes [
   n1 = RigVMFunctionEntryNode()
   RigVMFunction_ControlFlowBranch = &RigVMFunction_ControlFlowBranch(ExecuteContext=@n1.ExecuteContext, Condition=@n1.debug)
   RigVMFunctionReturnNode(ExecuteContext=@RigVMFunction_ControlFlowBranch.Completed)
   RigVMFunction_Sequence = RigVMAggregateNode(ExecuteContext=@RigVMFunction_ControlFlowBranch.True)
   RigVMFunction_MathQuaternionSwingTwist = &RigVMFunction_MathQuaternionSwingTwist(
    TwistAxis=(X=1,Y=0,Z=0)
    Input=&RigUnit_GetRelativeTransformForItem(Child=(Type=Bone,Name="thigh_l"), Parent=(Type=Null,Name="thigh_l_null"), bChildInitial=False, bParentInitial=False).RelativeTransform.Rotation
   )
   Inverse = /Script/RigVM.RigVMFunction_MathQuaternionInverse(Value=@RigVMFunction_MathQuaternionSwingTwist.Twist) {is=struct RigVMUnitNode, template="Inverse::Execute(in Value,out Result)"}
   &RigUnit_ModifyTransforms(
    ItemToModify=((Item=(Type=Bone,Name="thigh_twist_01_l"),Transform=(Rotation=(X=0,Y=0,Z=0,W=1),Translation=(X=0,Y=0,Z=0),Scale3D=(X=1,Y=1,Z=1))))
    Weight=1
    WeightMinimum=0
    ItemToModify.0.Transform.Rotation=@Inverse.Result
    ExecutePin=@RigVMFunction_Sequence.B
   )
   n2 = &RigUnit_GetRelativeTransformForItem(Child=(Type=Bone,Name="thigh_r"), Parent=(Type=Null,Name="thigh_r_null"), bChildInitial=False, bParentInitial=False)
   > &RigUnit_ModifyTransforms(
    ItemToModify=((Item=(Type=Bone,Name="thigh_twist_01_r"),Transform=(Rotation=(X=0,Y=0,Z=0,W=1),Translation=(X=0,Y=0,Z=0),Scale3D=(X=1,Y=1,Z=1))))
    Weight=1
    WeightMinimum=0
    ItemToModify.0.Transform.Rotation=&RigVMFunction_MathQuaternionInverse(Value=&RigVMFunction_MathQuaternionSwingTwist(TwistAxis=(X=1,Y=0,Z=0), Input=@n2.RelativeTransform.Rotation).Twist) {template="Inverse::Execute(in Value,out Result)"}.Result
   )
   n3 = &RigVMFunction_MathQuaternionSwingTwist(
    TwistAxis=(X=1,Y=0,Z=0)
    Input=&RigUnit_GetRelativeTransformForItem(Child=(Type=Bone,Name="foot_l"), Parent=(Type=Null,Name="foot_l_null"), bChildInitial=False, bParentInitial=False).RelativeTransform.Rotation
   )
   &RigUnit_ModifyTransforms(
    ItemToModify="((Item=(Type=Bone,Name=\"calf_twist_01_l\"),Transform=(Rotation=(X=0,Y=0,Z=0,W=1),Translation=(X=0,Y=0,Z=0),Scale3D=(X=1,Y=1,Z=1))),(Item=(Type=Bone,Name=\"calf_twist_02_l\"),Transform=(Rotation=(X=0,Y=0,Z=0,W=1),Translation=(X=0,Y=0,Z=0),Scale3D=(X=1,Y=1,Z=1) ...(truncated)"
    Weight=1
    WeightMinimum=0
    ItemToModify.1.Transform.Rotation=/Script/RigVM.RigVMFunction_MathQuaternionSlerp(B=(X=0,Y=0,Z=0,W=1), T=0.5, A=@n3.Twist) {is=struct RigVMUnitNode, template="Interpolate::Execute(in A,in B,in T,out Result)"}.Result
    ItemToModify.0.Transform.Rotation=@n3.Twist
    ExecutePin=@RigVMFunction_Sequence.D
   )
   n4 = &RigVMFunction_MathQuaternionSwingTwist(
    TwistAxis=(X=1,Y=0,Z=0)
    Input=&RigUnit_GetRelativeTransformForItem(Child=(Type=Bone,Name="foot_r"), Parent=(Type=Null,Name="foot_r_null"), bChildInitial=False, bParentInitial=False).RelativeTransform.Rotation
   )
   > &RigUnit_ModifyTransforms(
    ItemToModify="((Item=(Type=Bone,Name=\"calf_twist_01_r\"),Transform=(Rotation=(X=0,Y=0,Z=0,W=1),Translation=(X=0,Y=0,Z=0),Scale3D=(X=1,Y=1,Z=1))),(Item=(Type=Bone,Name=\"calf_twist_02_r\"),Transform=(Rotation=(X=0,Y=0,Z=0,W=1),Translation=(X=0,Y=0,Z=0),Scale3D=(X=1,Y=1,Z=1) ...(truncated)"
    Weight=1
    WeightMinimum=0
    ItemToModify.1.Transform.Rotation=&RigVMFunction_MathQuaternionSlerp(B=(X=0,Y=0,Z=0,W=1), T=0.5, A=@n4.Twist) {template="Interpolate::Execute(in A,in B,in T,out Result)"}.Result
    ItemToModify.0.Transform.Rotation=@n4.Twist
   )
   n5 = &RigVMFunction_MathQuaternionInverse(Value=&RigVMFunction_MathQuaternionSwingTwist(
     TwistAxis=(X=1,Y=0,Z=0)
     Input=&RigUnit_GetRelativeTransformForItem(Child=(Type=Bone,Name="upperarm_l"), Parent=(Type=Null,Name="upperarm_l_null"), bChildInitial=False, bParentInitial=False).RelativeTransform.Rotation
    ).Twist) {template="Inverse::Execute(in Value,out Result)"}
   &RigUnit_ModifyTransforms(
    ItemToModify="((Item=(Type=Bone,Name=\"upperarm_twist_01_l\"),Transform=(Rotation=(X=0,Y=0,Z=0,W=1),Translation=(X=0,Y=0,Z=0),Scale3D=(X=1,Y=1,Z=1))),(Item=(Type=Bone,Name=\"upperarm_twist_02_l\"),Transform=(Rotation=(X=0,Y=0,Z=0,W=1),Translation=(X=0,Y=0,Z=0),Scale3D=(X=1, ...(truncated)"
    Weight=1
    WeightMinimum=0
    ItemToModify.1.Transform.Rotation=&RigVMFunction_MathQuaternionSlerp(B=(X=0,Y=0,Z=0,W=1), T=0.5, A=@n5.Result) {template="Interpolate::Execute(in A,in B,in T,out Result)"}.Result
    ItemToModify.0.Transform.Rotation=@n5.Result
    ExecutePin=@RigVMFunction_Sequence.C
   )
   n6 = &RigVMFunction_MathQuaternionInverse(Value=&RigVMFunction_MathQuaternionSwingTwist(
     TwistAxis=(X=1,Y=0,Z=0)
     Input=&RigUnit_GetRelativeTransformForItem(Child=(Type=Bone,Name="upperarm_r"), Parent=(Type=Null,Name="upperarm_r_null"), bChildInitial=False, bParentInitial=False).RelativeTransform.Rotation
    ).Twist) {template="Inverse::Execute(in Value,out Result)"}
   Interpolate = &RigVMFunction_MathQuaternionSlerp(B=(X=0,Y=0,Z=0,W=1), T=0.5, A=@n6.Result) {template="Interpolate::Execute(in A,in B,in T,out Result)"}
   > &RigUnit_ModifyTransforms(
    ItemToModify="((Item=(Type=Bone,Name=\"upperarm_twist_01_r\"),Transform=(Rotation=(X=0,Y=0,Z=0,W=1),Translation=(X=0,Y=0,Z=0),Scale3D=(X=1,Y=1,Z=1))),(Item=(Type=Bone,Name=\"upperarm_twist_02_r\"),Transform=(Rotation=(X=0,Y=0,Z=0,W=1),Translation=(X=0,Y=0,Z=0),Scale3D=(X=1, ...(truncated)"
    Weight=1
    WeightMinimum=0
    ItemToModify.1.Transform.Rotation=@Interpolate.Result
    ItemToModify.0.Transform.Rotation=@n6.Result
   )
   RigUnit_ItemArray = &RigUnit_ItemArray(Items="((Type=Bone,Name=\"pelvis\"),(Type=Bone,Name=\"spine_01\"),(Type=Bone,Name=\"spine_02\"),(Type=Bone,Name=\"spine_03\"),(Type=Bone,Name=\"spine_04\"),(Type=Bone,Name=\"spine_05\"),(Type=Bone,Name=\"clavicle_l\"),(Type=Bone,Name=\"upperarm_l\"),(Type=Bone,Name=\"lowerarm_l\") ...(truncated)")
   DISPATCH_RigVMDispatch_ArrayIterator = RigVMDispatchNode(Array=@RigUnit_ItemArray.Items, ExecuteContext=@RigVMFunction_Sequence.A) {template="DISPATCH_RigVMDispatch_ArrayIterator(in Array,out Element,out Index,out Count,out Ratio)"}
   Concat = /Script/RigVM.RigVMFunction_NameConcat(B=_ctrl, A=@DISPATCH_RigVMDispatch_ArrayIterator.Element.Name) {is=struct RigVMUnitNode, template="Concat::Execute(in A,in B,out Result)"}
   > &RigUnit_SetTransform(
    Item=(Type=Bone,Name="spine_01")
    Weight=1
    bInitial=False
    Item.Name=@DISPATCH_RigVMDispatch_ArrayIterator.Element.Name
    Value=&RigUnit_GetTransform(Item=(Type=Control,Name="spine_01_ctrl"), bInitial=False, Item.Name=@Concat.Result).Transform
   )
   n7 = &RigVMFunction_MathQuaternionSwingTwist(
    TwistAxis=(X=1,Y=0,Z=0)
    Input=&RigUnit_GetRelativeTransformForItem(Child=(Type=Bone,Name="hand_l"), Parent=(Type=Null,Name="hand_l_null"), bChildInitial=False, bParentInitial=False).RelativeTransform.Rotation
   )
   &RigUnit_ModifyTransforms(
    ItemToModify="((Item=(Type=Bone,Name=\"lowerarm_twist_01_l\"),Transform=(Rotation=(X=0,Y=0,Z=0,W=1),Translation=(X=0,Y=0,Z=0),Scale3D=(X=1,Y=1,Z=1))),(Item=(Type=Bone,Name=\"lowerarm_twist_02_l\"),Transform=(Rotation=(X=0,Y=0,Z=0,W=1),Translation=(X=0,Y=0,Z=0),Scale3D=(X=1, ...(truncated)"
    Weight=1
    WeightMinimum=0
    ItemToModify.1.Transform.Rotation=&RigVMFunction_MathQuaternionSlerp(B=(X=0,Y=0,Z=0,W=1), T=0.5, A=@n7.Twist) {template="Interpolate::Execute(in A,in B,in T,out Result)"}.Result
    ItemToModify.0.Transform.Rotation=@n7.Twist
    ExecutePin=@RigVMFunction_Sequence.E
   )
   n8 = &RigVMFunction_MathQuaternionSwingTwist(
    TwistAxis=(X=1,Y=0,Z=0)
    Input=&RigUnit_GetRelativeTransformForItem(Child=(Type=Bone,Name="hand_r"), Parent=(Type=Null,Name="hand_r_null"), bChildInitial=False, bParentInitial=False).RelativeTransform.Rotation
   )
   > &RigUnit_ModifyTransforms(
    ItemToModify="((Item=(Type=Bone,Name=\"lowerarm_twist_01_r\"),Transform=(Rotation=(X=0,Y=0,Z=0,W=1),Translation=(X=0,Y=0,Z=0),Scale3D=(X=1,Y=1,Z=1))),(Item=(Type=Bone,Name=\"lowerarm_twist_02_r\"),Transform=(Rotation=(X=0,Y=0,Z=0,W=1),Translation=(X=0,Y=0,Z=0),Scale3D=(X=1, ...(truncated)"
    Weight=1
    WeightMinimum=0
    ItemToModify.1.Transform.Rotation=&RigVMFunction_MathQuaternionSlerp(B=(X=0,Y=0,Z=0,W=1), T=0.5, A=@n8.Twist) {template="Interpolate::Execute(in A,in B,in T,out Result)"}.Result
    ItemToModify.0.Transform.Rotation=@n8.Twist
   )
  ]}
 New Function_ContainedGraph {nodes [
   n1 = RigVMFunctionEntryNode()
   RigVMFunction_MathBoolToInteger = &RigVMFunction_MathBoolToInteger(Value=@n1.onlyTwist)
   RigVMFunction_MathQuaternionSwingTwist = &RigVMFunction_MathQuaternionSwingTwist(
    TwistAxis=@n1.twistAxis
    Input=&RigUnit_GetRelativeTransformForItem(
     Child=(Type=Bone,Name="None")
     Parent=(Type=Bone,Name="None")
     bChildInitial=False
     bParentInitial=False
     Child.Name=@n1.child
     Parent.Name=@n1.parent
     Parent.Type=@n1.parentType
    ).RelativeTransform.Rotation
   )
   RigVMFunction_MathQuaternionToEuler = &RigVMFunction_MathQuaternionToEuler(
    Value=RigVMDispatchNode(
     Values=((),())
     Index=@RigVMFunction_MathBoolToInteger.Result
     Values.0=@RigVMFunction_MathQuaternionSwingTwist.Swing
     Values.1=@RigVMFunction_MathQuaternionSwingTwist.Twist
    ) {template="DISPATCH_RigVMDispatch_SelectInt32(in Index,in Values,out Result)"}.Result
    RotationOrder=@n1.Rotation Order
   )
   > DISPATCH_RigVMDispatch_Print = RigVMDispatchNode(ScreenColor=(R=1,G=1,B=1,A=1), ScreenDuration=1.5, Enabled=@n1.debugRotation, Value=@RigVMFunction_MathQuaternionToEuler.Result) {template="DISPATCH_RigVMDispatch_Print(in Prefix,in Value,in Enabled,in ScreenDuration,in ScreenColor)"}
   Clamp = &RigVMFunction_MathIntClamp(Value=@n1.primAxis)
   DISPATCH_RigVMDispatch_SelectInt32 = RigVMDispatchNode(
    Values=(0,0,0)
    Index=@Clamp.Result
    Values.0=@RigVMFunction_MathQuaternionToEuler.Result.X
    Values.1=@RigVMFunction_MathQuaternionToEuler.Result.Y
    Values.2=@RigVMFunction_MathQuaternionToEuler.Result.Z
   ) {template="DISPATCH_RigVMDispatch_SelectInt32(in Index,in Values,out Result)"}
   Remap = &RigVMFunction_MathDoubleRemap(
    TargetMaximum=1
    TargetMinimum=0
    bClamp=False
    Value=@DISPATCH_RigVMDispatch_SelectInt32.Result
    SourceMaximum=@n1.max
    SourceMinimum=@n1.min
   )
   Multiply = &RigVMFunction_MathDoubleMul(A=&RigVMFunction_MathDoubleClamp(Value=@Remap.Result).Result, B=@n1.premult)
   > RigVMFunctionReturnNode(outputWeight=@Multiply.Result)
  ]}
 New Function_ContainedGraph {nodes [
   DeltaRotation_1 = RigVMFunctionReferenceNode(
    ExecuteContext=()
    Rotation Order=ZYX
    child=head
    debugRotation=false
    max=-60
    min=-40
    parent=head_null
    parentType=Null
    premult=1
    primAxis=2
    twistAxis=(X=1,Y=0,Z=0)
   )
   > &RigUnit_ModifyTransforms(
    ItemToModify="((Item=(Type=Bone,Name=\"base_hood_c\"),Transform=(Rotation=(X=0,Y=-0.342,Z=0,W=0.9397),Translation=(X=0,Y=0,Z=0),Scale3D=(X=1,Y=1,Z=1))),(Item=(Type=Bone,Name=\"base_hood_3_l\"),Transform=(Rotation=(X=0,Y=-0.2588,Z=0,W=0.9659),Translation=(X=0,Y=0,Z=0),Scale3 ...(truncated)"
    WeightMinimum=0
    Weight=@DeltaRotation_1.outputWeight
   )
   RigVMFunctionEntryNode()
   > RigVMCollapseNode()
   > RigVMFunction_Sequence = RigVMAggregateNode()
   DeltaRotation_2 = RigVMFunctionReferenceNode(
    Rotation Order=ZYX
    child=head
    debugRotation=false
    max=-20
    min=-90
    onlyTwist=true
    parent=head_null
    parentType=Null
    premult=1
    primAxis=0
    twistAxis=(X=1,Y=0,Z=0)
    ExecuteContext=@RigVMFunction_Sequence.A
   )
   DeltaRotation_2_1 = RigVMFunctionReferenceNode(
    Rotation Order=ZYX
    child=head
    debugRotation=false
    max=20
    min=90
    onlyTwist=true
    parent=head_null
    parentType=Null
    premult=1
    primAxis=0
    twistAxis=(X=1,Y=0,Z=0)
    ExecuteContext=@RigVMFunction_Sequence.B
   )
   DeltaRotation = RigVMFunctionReferenceNode(
    Rotation Order=ZYX
    child=head
    debugRotation=false
    max=-40
    min=-12
    parent=head_null
    parentType=Null
    premult=1
    primAxis=2
    twistAxis=(X=1,Y=0,Z=0)
    ExecuteContext=@RigVMFunction_Sequence.C
   )
   Minimum = /Script/RigVM.RigVMFunction_MathDoubleMin(A=@DeltaRotation_2.outputWeight, B=@DeltaRotation_2_1.outputWeight) {is=struct RigVMUnitNode, template="Minimum::Execute(in A,in B,out Result)"}
   Multiply = &RigVMFunction_MathDoubleMul(B=@DeltaRotation.outputWeight, A=@Minimum.Result)
   > &RigUnit_ModifyTransforms(
    ItemToModify="((Item=(Type=Bone,Name=\"base_hood_c\"),Transform=(Rotation=(X=0,Y=-0.5736,Z=0,W=0.8192),Translation=(X=-1.2,Y=0,Z=0),Scale3D=(X=1,Y=1,Z=1))),(Item=(Type=Bone,Name=\"base_hood_3_l\"),Transform=(Rotation=(X=0,Y=-0.2588,Z=0,W=0.9659),Translation=(X=-1,Y=0,Z=0),S ...(truncated)"
    WeightMinimum=0
    Weight=@Multiply.Result
   )
   > n1 = &RigUnit_SphericalPoseReader(
    ActiveRegionScaleFactors=(PositiveWidth=1,NegativeWidth=1,PositiveHeight=1,NegativeHeight=1)
    ActiveRegionSize=0.1
    Debug=(bDrawDebug=false,bDraw2D=False,bDrawLocalAxes=False,DebugScale=25,DebugSegments=20,DebugThickness=0.25)
    DriverAxis=(X=1,Y=0,Z=0)
    DriverItem=(Type=Bone,Name="head")
    FalloffRegionScaleFactors=(PositiveWidth=1,NegativeWidth=1,PositiveHeight=1,NegativeHeight=1)
    FalloffSize=0.2
    FlipHeightScaling=False
    FlipWidthScaling=False
    OptionalParentItem=(Type=Bone,Name="spine_05")
    RotationOffset=(X=-60,Y=0,Z=0)
   )
   > &RigUnit_ModifyTransforms(
    ItemToModify="((Item=(Type=Bone,Name=\"base_hood_3_r\"),Transform=(Rotation=(X=0,Y=-0.4226,Z=0,W=0.9063),Translation=(X=0,Y=0,Z=0),Scale3D=(X=1,Y=1,Z=1))),(Item=(Type=Bone,Name=\"base_hood_c\"),Transform=(Rotation=(X=0,Y=-0.5,Z=0,W=0.866),Translation=(X=0,Y=0,Z=0),Scale3D=( ...(truncated)"
    WeightMinimum=0
    Weight=@n1.OutputParam
   )
   > RigVMFunctionReturnNode()
  ]}
 New Function_ContainedGraph {nodes [
   RigVMFunctionEntryNode()
   > DeltaRotation = RigVMFunctionReferenceNode(Rotation Order=ZXY, child=thigh_l, max=-100, min=-8, parent=thigh_l_null, parentType=Null, premult=1, primAxis=2, twistAxis=(X=1,Y=0,Z=0))
   > RigVMVariableNode(Value=@DeltaRotation.outputWeight) {variable=thighRzNeg}
   RigVMFunction_AnimRichCurve = /Script/RigVM.RigVMFunction_AnimRichCurve(Curve="(EditorCurveData=(Keys=((InterpMode=RCIM_Cubic,TangentMode=RCTM_Auto,TangentWeightMode=RCTWM_WeightedNone,Time=0,Value=0,ArriveTangent=-0.1644,ArriveTangentWeight=0,LeaveTangent=0,LeaveTangentWeight=0),(InterpMode=RCIM_Cubic,TangentMode=RCTM_User,TangentWe ...(truncated)") {is=struct RigVMUnitNode}
   RigVMFunction_AnimEvalRichCurve = &RigVMFunction_AnimEvalRichCurve(SourceMinimum=0, TargetMinimum=0, Value=@DeltaRotation.outputWeight, Curve=@RigVMFunction_AnimRichCurve.Curve)
   > &RigUnit_ModifyTransforms(
    ItemToModify="((Item=(Type=Bone,Name=\"baseW_tailCoat_front_l_1\"),Transform=(Rotation=(X=0,Y=-0.6428,Z=0,W=0.766),Translation=(X=0,Y=0,Z=0),Scale3D=(X=1,Y=1,Z=1))),(Item=(Type=Bone,Name=\"base_tailCoat_front_l_1\"),Transform=(Rotation=(X=-0.1736,Y=0,Z=0,W=0.9848),Translati ...(truncated)"
    WeightMinimum=0
    Weight=@RigVMFunction_AnimEvalRichCurve.Result
   )
   > DeltaRotation_1 = RigVMFunctionReferenceNode(Rotation Order=ZXY, child=thigh_r, max=-100, min=-8, parent=thigh_r_null, parentType=Null, premult=1, primAxis=2, twistAxis=(X=1,Y=0,Z=0))
   > &RigUnit_ModifyTransforms(
    ItemToModify="((Item=(Type=Bone,Name=\"baseW_tailCoat_front_r_1\"),Transform=(Rotation=(X=0,Y=-0.6428,Z=0,W=0.766),Translation=(X=0,Y=0,Z=0),Scale3D=(X=1,Y=1,Z=1))),(Item=(Type=Bone,Name=\"base_tailCoat_front_r_1\"),Transform=(Rotation=(X=0.1736,Y=0,Z=0,W=0.9848),Translatio ...(truncated)"
    WeightMinimum=0
    Weight=&RigVMFunction_AnimEvalRichCurve(SourceMinimum=0, TargetMinimum=0, Value=@DeltaRotation_1.outputWeight, Curve=@RigVMFunction_AnimRichCurve.Curve).Result
   )
   > DeltaRotation_2 = RigVMFunctionReferenceNode(Rotation Order=ZXY, child=thigh_l, max=64, min=0, parent=thigh_l_null, parentType=Null, premult=1, primAxis=2, twistAxis=(X=1,Y=0,Z=0))
   > DeltaRotation_2_1 = RigVMFunctionReferenceNode(Rotation Order=ZXY, child=thigh_r, max=64, min=35, parent=thigh_r_null, parentType=Null, premult=1, primAxis=2, twistAxis=(X=1,Y=0,Z=0))
   Maximum = &RigVMFunction_MathDoubleMax(A=@DeltaRotation_2_1.outputWeight, B=@DeltaRotation_2.outputWeight)
   > &RigUnit_ModifyTransforms(
    ItemToModify=((Item=(Type=Bone,Name="base_tailCoat_back_c_1"),Transform=(Rotation=(X=0,Y=-0.342,Z=0,W=0.9397),Translation=(X=0,Y=0,Z=0),Scale3D=(X=1,Y=1,Z=1))))
    WeightMinimum=0
    Weight=@Maximum.Result
   )
   > DeltaRotation_3 = RigVMFunctionReferenceNode(
    Rotation Order=ZYX
    child=clavicle_l
    debugRotation=false
    max=-40
    parent=clavicle_l_null
    parentType=Null
    premult=0.4
    primAxis=1
    twistAxis=(X=1,Y=0,Z=0)
   )
   > &RigUnit_ModifyTransforms(
    ItemToModify="((Item=(Type=Bone,Name=\"deform_lat_l\"),Transform=(Rotation=(X=0,Y=0,Z=0,W=1),Translation=(X=12,Y=0,Z=0),Scale3D=(X=1,Y=1,Z=1))),(Item=(Type=Bone,Name=\"deform_lat_l_base\"),Transform=(Rotation=(X=0,Y=-0.0872,Z=0,W=0.9962),Translation=(X=0,Y=0,Z=0.5),Scale3D= ...(truncated)"
    WeightMinimum=0
    Weight=@DeltaRotation_3.outputWeight
   )
   > DeltaRotation_3_1 = RigVMFunctionReferenceNode(
    Rotation Order=ZYX
    child=clavicle_r
    debugRotation=false
    max=-40
    parent=clavicle_r_null
    parentType=Null
    premult=0.4
    primAxis=1
    twistAxis=(X=1,Y=0,Z=0)
   )
   > &RigUnit_ModifyTransforms(
    ItemToModify="((Item=(Type=Bone,Name=\"deform_lat_r\"),Transform=(Rotation=(X=0,Y=0,Z=0,W=1),Translation=(X=12,Y=0,Z=0),Scale3D=(X=1,Y=1,Z=1))),(Item=(Type=Bone,Name=\"deform_lat_r_base\"),Transform=(Rotation=(X=0,Y=0.0872,Z=0,W=0.9962),Translation=(X=0,Y=0,Z=-0.5),Scale3D= ...(truncated)"
    WeightMinimum=0
    Weight=@DeltaRotation_3_1.outputWeight
   )
   > RigVMCollapseNode()
   > DeltaRotation_4 = RigVMFunctionReferenceNode(
    Rotation Order=ZYX
    child=head
    debugRotation=false
    max=20
    min=10
    onlyTwist=false
    parent=head_null
    parentType=Null
    premult=1
    primAxis=2
    twistAxis=(X=1,Y=0,Z=0)
   )
   > DeltaRotation_4_1 = RigVMFunctionReferenceNode(
    Rotation Order=ZYX
    child=head
    debugRotation=false
    max=-10
    onlyTwist=true
    parent=head_null
    parentType=Null
    premult=1
    twistAxis=(X=1,Y=0,Z=0)
   )
   Multiply = &RigVMFunction_MathDoubleMul(A=@DeltaRotation_4_1.outputWeight, B=@DeltaRotation_4.outputWeight)
   > &RigUnit_ModifyTransforms(
    ItemToModify="((Item=(Type=Bone,Name=\"base_internalNeckCollar_1_l\"),Transform=(Rotation=(X=0,Y=0,Z=-0.1736,W=0.9848),Translation=(X=-1.5,Y=1,Z=0),Scale3D=(X=1,Y=1,Z=1))),(Item=(Type=Bone,Name=\"base_internalNeckCollar_1_c\"),Transform=(Rotation=(X=0,Y=0,Z=-0.0872,W=0.9962 ...(truncated)"
    WeightMinimum=0
    Weight=@Multiply.Result
   )
   > DeltaRotation_4_1_1 = RigVMFunctionReferenceNode(
    Rotation Order=ZYX
    child=head
    debugRotation=false
    max=10
    onlyTwist=true
    parent=head_null
    parentType=Null
    premult=1
    twistAxis=(X=1,Y=0,Z=0)
   )
   > &RigUnit_ModifyTransforms(
    ItemToModify="((Item=(Type=Bone,Name=\"base_internalNeckCollar_1_r\"),Transform=(Rotation=(X=0,Y=0,Z=-0.1736,W=0.9848),Translation=(X=-1.5,Y=1,Z=0),Scale3D=(X=1,Y=1,Z=1))),(Item=(Type=Bone,Name=\"base_internalNeckCollar_1_c\"),Transform=(Rotation=(X=0,Y=0,Z=-0.0872,W=0.9962 ...(truncated)"
    WeightMinimum=0
    Weight=&RigVMFunction_MathDoubleMul(A=@DeltaRotation_4_1_1.outputWeight, B=@DeltaRotation_4.outputWeight).Result
   )
   > RigVMFunctionReturnNode()
  ]}
 New Function_ContainedGraph {nodes [
   RigVMFunctionEntryNode()
   > &RigUnit_ParentConstraint(
    AdvancedSettings=(RotationOrderForFilter=XZY)
    Child=(Type=Bone,Name="base_radio")
    Parents="((Item=(Type=Bone,Name=\"spine_01\"),Weight=0.0058),(Item=(Type=Bone,Name=\"deform_glute_r\"),Weight=0.0404),(Item=(Type=Bone,Name=\"deform_groin_r\"),Weight=0.1232),(Item=(Type=Bone,Name=\"thigh_twist_01_r\"),Weight=0.2869),(Item=(Type=Bone,Name=\"pelvis\"),Weight= ...(truncated)"
    Weight=1
   )
   > DeltaRotation = RigVMFunctionReferenceNode(Rotation Order=ZXY, child=spine_01, max=-20, parent=pelvis, parentType=Bone, premult=1, primAxis=1, twistAxis=(X=1,Y=0,Z=0))
   > &RigUnit_ModifyTransforms(
    ItemToModify=((Item=(Type=Bone,Name="base_radioAntenna"),Transform=(Rotation=(X=0,Y=-0.1736,Z=0,W=0.9848),Translation=(X=0,Y=0,Z=0),Scale3D=(X=1,Y=1,Z=1))))
    WeightMinimum=0
    Weight=@DeltaRotation.outputWeight
   )
   > DeltaRotation_1 = RigVMFunctionReferenceNode(Rotation Order=ZYX, child=thigh_r, max=-70, min=0, parent=thigh_r_null, parentType=Null, premult=1, primAxis=1, twistAxis=(X=1,Y=0,Z=0))
   > &RigUnit_ModifyTransforms(
    ItemToModify="((Item=(Type=Bone,Name=\"base_radio\"),Transform=(Rotation=(X=-0.1585,Y=-0.1894,Z=-0.031,W=0.9685),Translation=(X=0.8241,Y=-1.6128,Z=0.4884),Scale3D=(X=1,Y=1,Z=1))),(Item=(Type=Bone,Name=\"base_radioWire\"),Transform=(Rotation=(X=0,Y=0.6624,Z=0,W=0.7491),Trans ...(truncated)"
    WeightMinimum=0
    Weight=@DeltaRotation_1.outputWeight
   )
   > n1 = &RigUnit_SphericalPoseReader(
    ActiveRegionScaleFactors=(PositiveWidth=1,NegativeWidth=1,PositiveHeight=1,NegativeHeight=1)
    ActiveRegionSize=0.05
    Debug=(bDrawDebug=false,bDraw2D=False,bDrawLocalAxes=False,DebugScale=25,DebugSegments=20,DebugThickness=0.25)
    DriverAxis=(X=1,Y=0,Z=0)
    DriverItem=(Type=Bone,Name="thigh_r")
    FalloffRegionScaleFactors=(PositiveWidth=1,NegativeWidth=1,PositiveHeight=1,NegativeHeight=1)
    FalloffSize=0.25
    FlipHeightScaling=False
    FlipWidthScaling=False
    OptionalParentItem=(Type=Bone,Name="None")
    RotationOffset=(X=-30,Y=-5,Z=0)
   )
   > &RigUnit_ModifyTransforms(
    ItemToModify="((Item=(Type=Bone,Name=\"base_radioWire\"),Transform=(Rotation=(X=-0.3826,Y=0.0091,Z=-0.0038,W=0.9239),Translation=(X=0,Y=0.55,Z=-2.35),Scale3D=(X=1,Y=1,Z=1))),(Item=(Type=Bone,Name=\"dyn_radio\"),Transform=(Rotation=(X=-0.1403,Y=0,Z=0,W=0.9901),Translation=(X ...(truncated)"
    WeightMinimum=0
    Weight=@n1.OutputParam
   )
   > RigVMFunctionReturnNode()
  ]}
 New Function_ContainedGraph {nodes [
   RigVMFunctionEntryNode()
   > DeltaRotation = RigVMFunctionReferenceNode(
    Rotation Order=ZYX
    child=calf_l
    max=180
    min=120
    onlyTwist=false
    parent=calf_l_null
    parentType=Null
    premult=1
    primAxis=2
    twistAxis=(X=1,Y=0,Z=0)
   )
   > &RigUnit_ModifyTransforms(
    ItemToModify="((Item=(Type=Bone,Name=\"deform_knee_in_l\"),Transform=(Rotation=(X=0,Y=0,Z=0,W=1),Translation=(X=0,Y=-15,Z=0),Scale3D=(X=2,Y=1,Z=1))),(Item=(Type=Bone,Name=\"calf_twist_02_l\"),Transform=(Rotation=(X=0,Y=0,Z=0,W=1),Translation=(X=0,Y=4,Z=0),Scale3D=(X=1,Y=1,Z ...(truncated)"
    WeightMinimum=0
    Weight=@DeltaRotation.outputWeight
   )
   > DeltaRotation_1 = RigVMFunctionReferenceNode(
    Rotation Order=ZYX
    child=calf_r
    max=180
    min=120
    onlyTwist=false
    parent=calf_r_null
    parentType=Null
    premult=1
    primAxis=2
    twistAxis=(X=1,Y=0,Z=0)
   )
   > &RigUnit_ModifyTransforms(
    ItemToModify="((Item=(Type=Bone,Name=\"deform_knee_in_r\"),Transform=(Rotation=(X=0,Y=0,Z=0,W=1),Translation=(X=0,Y=15,Z=0),Scale3D=(X=2,Y=1,Z=1))),(Item=(Type=Bone,Name=\"calf_twist_02_r\"),Transform=(Rotation=(X=0,Y=0,Z=0,W=1),Translation=(X=0,Y=-4,Z=0),Scale3D=(X=1,Y=1,Z ...(truncated)"
    WeightMinimum=0
    Weight=@DeltaRotation_1.outputWeight
   )
   > RigVMFunctionReturnNode()
   &RigUnit_SphericalPoseReader(
    ActiveRegionScaleFactors=(PositiveWidth=0.1,NegativeWidth=0.1,PositiveHeight=0.1,NegativeHeight=0.1)
    ActiveRegionSize=0.1
    Debug=(bDrawDebug=false,bDraw2D=false,DebugScale=25,DebugSegments=20,DebugThickness=0.25)
    DriverAxis=(X=1,Y=0,Z=0)
    DriverItem=(Type=Bone,Name="calf_r")
    ExecutePin=()
    FalloffRegionScaleFactors=(PositiveWidth=0,NegativeWidth=0.75,PositiveHeight=0.75,NegativeHeight=0.75)
    FalloffSize=0.45
    OptionalParentItem=(Type=Bone,Name="None")
    RotationOffset=(X=90,Y=180,Z=90)
   )
   &RigUnit_SphericalPoseReader(
    ActiveRegionScaleFactors=(PositiveWidth=0.1,NegativeWidth=0.1,PositiveHeight=0.1,NegativeHeight=0.1)
    ActiveRegionSize=0.1
    Debug=(bDrawDebug=true,bDraw2D=false,DebugScale=25,DebugSegments=20,DebugThickness=0.25)
    DriverAxis=(X=-1,Y=0,Z=0)
    DriverItem=(Type=Bone,Name="calf_l")
    ExecutePin=()
    FalloffRegionScaleFactors=(PositiveWidth=0,NegativeWidth=0.75,PositiveHeight=0.75,NegativeHeight=0.75)
    FalloffSize=0.45
    OptionalParentItem=(Type=Bone,Name="None")
    RotationOffset=(X=-90,Y=0,Z=90)
   )
  ]}
 New Function_ContainedGraph {nodes [
   RigVMFunctionEntryNode()
   > RigVMCollapseNode()
   > DeltaRotation_3 = RigVMFunctionReferenceNode(
    Rotation Order=ZYX
    child=clavicle_l
    debugRotation=false
    max=-50
    parent=clavicle_l_null
    parentType=Null
    premult=1
    primAxis=1
    twistAxis=(X=1,Y=0,Z=0)
   )
   RigVMFunction_AnimRichCurve = &RigVMFunction_AnimRichCurve(Curve="(EditorCurveData=(Keys=((InterpMode=RCIM_Cubic,TangentMode=RCTM_User,TangentWeightMode=RCTWM_WeightedNone,Time=0,Value=0,ArriveTangent=-0.0103,ArriveTangentWeight=0,LeaveTangent=-0.0103,LeaveTangentWeight=0),(InterpMode=RCIM_Cubic,TangentMode=RCTM_User,Tan ...(truncated)")
   RigVMFunction_AnimEvalRichCurve = &RigVMFunction_AnimEvalRichCurve(SourceMinimum=0, TargetMinimum=0, Value=@DeltaRotation_3.outputWeight, Curve=@RigVMFunction_AnimRichCurve.Curve)
   > &RigUnit_ModifyTransforms(
    ItemToModify="((Item=(Type=Bone,Name=\"stk_sTube_a_2_l\"),Transform=(Rotation=(X=0,Y=0,Z=0,W=1),Translation=(X=0,Y=-1.5,Z=2.2),Scale3D=(X=1,Y=1,Z=1))),(Item=(Type=Bone,Name=\"stk_sTube_b_2_l\"),Transform=(Rotation=(X=0,Y=0,Z=0,W=1),Translation=(X=0,Y=-1.2,Z=0.8),Scale3D=(X= ...(truncated)"
    WeightMinimum=0
    Weight=@RigVMFunction_AnimEvalRichCurve.Result
   )
   > DeltaRotation_3_1 = RigVMFunctionReferenceNode(
    Rotation Order=ZYX
    child=clavicle_r
    debugRotation=false
    max=-50
    parent=clavicle_r_null
    parentType=Null
    premult=1
    primAxis=1
    twistAxis=(X=1,Y=0,Z=0)
   )
   > &RigUnit_ModifyTransforms(
    ItemToModify="((Item=(Type=Bone,Name=\"stk_sTube_a_2_r\"),Transform=(Rotation=(X=0,Y=0,Z=0,W=1),Translation=(X=0,Y=1.5,Z=2.2),Scale3D=(X=1,Y=1,Z=1))),(Item=(Type=Bone,Name=\"stk_sTube_b_2_r\"),Transform=(Rotation=(X=0,Y=0,Z=0,W=1),Translation=(X=0,Y=1.2,Z=0.8),Scale3D=(X=1, ...(truncated)"
    WeightMinimum=0
    Weight=&RigVMFunction_AnimEvalRichCurve(SourceMinimum=0, TargetMinimum=0, Value=@DeltaRotation_3_1.outputWeight, Curve=@RigVMFunction_AnimRichCurve.Curve).Result
   )
   > n1 = &RigUnit_SphericalPoseReader(
    ActiveRegionScaleFactors=(PositiveWidth=1,NegativeWidth=1,PositiveHeight=1.5,NegativeHeight=1.5)
    ActiveRegionSize=0.1
    Debug=(bDrawDebug=false,bDraw2D=false,bDrawLocalAxes=false,DebugScale=25,DebugSegments=20,DebugThickness=0.25)
    DriverAxis=(X=1,Y=0,Z=0)
    DriverItem=(Type=Bone,Name="upperarm_l")
    FalloffRegionScaleFactors=(PositiveWidth=1.5,NegativeWidth=1.5,PositiveHeight=2,NegativeHeight=2)
    FalloffSize=0.2
    FlipHeightScaling=False
    FlipWidthScaling=False
    OptionalParentItem=(Type=Bone,Name="None")
    RotationOffset=(X=56,Y=18,Z=0)
   )
   > &RigUnit_ModifyTransforms(
    ItemToModify="((Item=(Type=Bone,Name=\"stk_sTube_a_1_l\"),Transform=(Rotation=(X=0,Y=0,Z=0,W=1),Translation=(X=0,Y=0.75,Z=0),Scale3D=(X=1,Y=1,Z=1))),(Item=(Type=Bone,Name=\"stk_sTube_b_1_l\"),Transform=(Rotation=(X=0,Y=0,Z=-0.061,W=0.9981),Translation=(X=-0.2,Y=1.1,Z=0),Sca ...(truncated)"
    WeightMinimum=0
    Weight=@n1.OutputParam
   )
   > n2 = &RigUnit_SphericalPoseReader(
    ActiveRegionScaleFactors=(PositiveWidth=1,NegativeWidth=1,PositiveHeight=1,NegativeHeight=2)
    ActiveRegionSize=0.1
    Debug=(bDrawDebug=false,bDraw2D=false,bDrawLocalAxes=false,DebugScale=25,DebugSegments=20,DebugThickness=0.25)
    DriverAxis=(X=-1,Y=0,Z=0)
    DriverItem=(Type=Bone,Name="upperarm_r")
    FalloffRegionScaleFactors=(PositiveWidth=1,NegativeWidth=1,PositiveHeight=1,NegativeHeight=2)
    FalloffSize=0.25
    FlipHeightScaling=False
    FlipWidthScaling=False
    OptionalParentItem=(Type=Bone,Name="None")
    RotationOffset=(X=-100,Y=20,Z=10)
   )
   > &RigUnit_ModifyTransforms(
    ItemToModify="((Item=(Type=Bone,Name=\"stk_sTube_a_1_r\"),Transform=(Rotation=(X=0,Y=0,Z=0,W=1),Translation=(X=0,Y=-1.85,Z=-0.45),Scale3D=(X=1,Y=1,Z=1))),(Item=(Type=Bone,Name=\"stk_sTube_b_1_r\"),Transform=(Rotation=(X=0,Y=0,Z=0,W=1),Translation=(X=0,Y=-1,Z=0),Scale3D=(X=1 ...(truncated)"
    WeightMinimum=0
    Weight=@n2.OutputParam
   )
   > RigVMFunctionReturnNode()
   n3 = &RigUnit_SphericalPoseReader(
    ActiveRegionScaleFactors=(PositiveWidth=1,NegativeWidth=1,PositiveHeight=1.5,NegativeHeight=1.5)
    ActiveRegionSize=0.1
    Debug=(bDrawDebug=false,bDraw2D=false,bDrawLocalAxes=false,DebugScale=25,DebugSegments=20,DebugThickness=0.25)
    DriverAxis=(X=1,Y=0,Z=0)
    DriverItem=(Type=Bone,Name="upperarm_l")
    ExecutePin=()
    FalloffRegionScaleFactors=(PositiveWidth=1.5,NegativeWidth=1.5,PositiveHeight=2,NegativeHeight=2)
    FalloffSize=0.2
    FlipHeightScaling=False
    FlipWidthScaling=False
    OptionalParentItem=(Type=Bone,Name="None")
    RotationOffset=(X=56,Y=18,Z=0)
   )
   > &RigUnit_ModifyTransforms(
    ItemToModify="((Item=(Type=Bone,Name=\"stk_sTube_a_1_l\"),Transform=(Rotation=(X=0,Y=0,Z=0,W=1),Translation=(X=0,Y=0.75,Z=0),Scale3D=(X=1,Y=1,Z=1))),(Item=(Type=Bone,Name=\"stk_sTube_b_1_l\"),Transform=(Rotation=(X=0,Y=0,Z=-0.061,W=0.9981),Translation=(X=-0.2,Y=1.1,Z=0),Sca ...(truncated)"
    WeightMinimum=0
    Weight=@n3.OutputParam
   )
   > n4 = &RigUnit_SphericalPoseReader(
    ActiveRegionScaleFactors=(PositiveWidth=1,NegativeWidth=1,PositiveHeight=1.5,NegativeHeight=1.5)
    ActiveRegionSize=0.1
    Debug=(bDrawDebug=false,bDraw2D=false,bDrawLocalAxes=false,DebugScale=25,DebugSegments=20,DebugThickness=0.25)
    DriverAxis=(X=-1,Y=0,Z=0)
    DriverItem=(Type=Bone,Name="upperarm_r")
    FalloffRegionScaleFactors=(PositiveWidth=1.5,NegativeWidth=1.5,PositiveHeight=2,NegativeHeight=2)
    FalloffSize=0.2
    FlipHeightScaling=False
    FlipWidthScaling=False
    OptionalParentItem=(Type=Bone,Name="None")
    RotationOffset=(X=-56,Y=-162,Z=0)
   )
   > &RigUnit_ModifyTransforms(
    ItemToModify="((Item=(Type=Bone,Name=\"stk_sTube_a_1_l\"),Transform=(Rotation=(X=0,Y=0,Z=0,W=1),Translation=(X=0,Y=-0.75,Z=0),Scale3D=(X=1,Y=1,Z=1))),(Item=(Type=Bone,Name=\"stk_sTube_b_1_l\"),Transform=(Rotation=(X=0,Y=0,Z=0.061,W=0.9981),Translation=(X=-0.2,Y=-1.1,Z=0),Sc ...(truncated)"
    WeightMinimum=0
    Weight=@n4.OutputParam
   )
  ]}
 RigVMFunctionLibrary {nodes [
   DeltaRotation = RigVMCollapseNode(ExecuteContext=(), Rotation Order=ZYX, premult=1, twistAxis=(X=1,Y=0,Z=0))
   Hood = RigVMCollapseNode(ExecuteContext=())
   Jacket = RigVMCollapseNode(ExecuteContext=())
   Misc = RigVMCollapseNode(ExecuteContext=())
   Tubes = RigVMCollapseNode(ExecuteContext=())
   controllers = RigVMCollapseNode(ExecuteContext=(), debug=False)
   customKneeSetup = RigVMCollapseNode(ExecuteContext=())
  ]}
 RigVMFunction_Sequence_1_ContainedGraph {nodes [
   n1 = RigVMFunctionEntryNode()
   RigVMFunction_Sequence = &RigVMFunction_Sequence(ExecuteContext=@n1.ExecuteContext)
   n2 = &RigVMFunction_Sequence(ExecuteContext=@RigVMFunction_Sequence.B)
   RigVMFunctionReturnNode(A=@RigVMFunction_Sequence.A, B=@n2.A, C=@n2.B)
  ]}
 RigVMFunction_Sequence_2_ContainedGraph {nodes [
   n1 = RigVMFunctionEntryNode()
   RigVMFunction_Sequence = &RigVMFunction_Sequence(ExecuteContext=@n1.ExecuteContext)
   n2 = &RigVMFunction_Sequence(ExecuteContext=@RigVMFunction_Sequence.B)
   n3 = &RigVMFunction_Sequence(ExecuteContext=@n2.B)
   n4 = &RigVMFunction_Sequence(ExecuteContext=@n3.B)
   RigVMFunctionReturnNode(A=@RigVMFunction_Sequence.A, B=@n2.A, C=@n3.A, D=@n4.A, E=@n4.B)
  ]}
 RigVMModel {nodes [
   RigUnit_BeginExecution = &RigUnit_BeginExecution()
   controllers = RigVMFunctionReferenceNode(debug=false, ExecuteContext=@RigUnit_BeginExecution.ExecutePin)
   > TwistsFix = RigVMCollapseNode()
   > FN_DeformRig_FNC = RigVMFunctionReferenceNode(
    DeformRig_ Curve="(deformRig_l_upperarm_up_alpha_72_#1=False,deformRig_r_upperarm_up_alpha_73_#2=False,deformRig_l_upperarm_fr_alpha_74_#3=False,deformRig_r_upperarm_fr_alpha_75_CB51B8 ...(truncated)"
    DeformRig_Bool="(upperarm_1_#4=True,elbow_3_#5=True,wrist_7_#6=True,groin_8_#7=True,glute_17_#8=True,knee_9_40C408324CA261 ...(truncated)"
    DeformRig_Debug=(upperarm_debug_18_#4=false,wrist_debug_20_#6=False,lat_debug_26_#9=False,groinGlute_debug_27_#7=False)
    DeformRig_Value="(upperarm_up_18_#10=(Rotation=(X=0,Y=0,Z=0,W=1),Translation=(X=0,Y=0,Z=1),Scale3D=(X=1,Y=1,Z=1)),upperarm_fwd_fr_19_#11=(Rotation=(X=0,Y=0,Z=0,W=1),Translation=(X=0,Y=14,Z=0),Scale3D=(X=0.5,Y=1,Z=1) ...(truncated)"
   )
   > customKneeSetup = RigVMFunctionReferenceNode()
   > Hood = RigVMFunctionReferenceNode()
   > Jacket = RigVMFunctionReferenceNode()
   > Tubes = RigVMFunctionReferenceNode()
   > Misc = RigVMFunctionReferenceNode()
  ]}
]
hierarchy [
 root {loc=<0 0 0>, rot=<0 0 0 1>}
 pelvis {parent=root, loc=<0 -0.7071 93.1191>, rot=<0.0127 0.707 -0.0127 -0.707>}
 spine_01 {parent=pelvis, loc=<6.7494 -0.6388 0>, rot=<0 0 0.034 -0.9994>}
 spine_02 {parent=spine_01, loc=<7.2518 -0.035 0>, rot=<0 0 0.0238 -0.9997>}
 spine_03 {parent=spine_02, loc=<6.0104 0 0>, rot=<0 0 0.0149 -0.9999>}
 spine_04 {parent=spine_03, loc=<8.1024 0 0>, rot=<0 0 -0.0132 -0.9999>}
 spine_05 {parent=spine_04, loc=<10.0146 0 0>, rot=<0 0 -0.0015 -1>}
 clavicle_l {parent=spine_05, loc=<4.3575 -0.1466 -2.3223>, rot=<0.0465 -0.7289 -0.009 -0.6829>}
 upperarm_l {parent=clavicle_l, loc=<15.5111 0 0>, rot=<0.0226 -0.3234 -0.0088 -0.9459>}
 lowerarm_l {parent=upperarm_l, loc=<26.6514 0 0>, rot=<0.0391 -0.0099 -0.2453 -0.9686>}
 hand_l {parent=lowerarm_l, loc=<22.3491 0 0>, rot=<-0.7081 -0.0507 0.0975 -0.6975>}
 index_metacarpal_l {parent=hand_l, loc=<1.6535 0.0183 -2.96>, rot=<0.0158 -0.0124 -0.0173 -0.9996>}
 index_01_l {parent=index_metacarpal_l, loc=<8.3008 0.0357 -0.0524>, rot=<-0.0044 0.0485 0.0924 -0.9945>}
 index_02_l {parent=index_01_l, loc=<3.3587 -0.3004 0.0452>, rot=<-0.002 0.0027 0.3225 -0.9466>}
 index_03_l {parent=index_02_l, loc=<2.5578 -0.123 -0.0252>, rot=<-0.002 0.0067 0.2558 -0.9667>}
 middle_metacarpal_l {parent=hand_l, loc=<2.1856 -0.1725 -0.5984>, rot=<0.006 0.0168 -0.0329 -0.9993>}
 middle_01_l {parent=middle_metacarpal_l, loc=<7.3275 -0.2447 0.0129>, rot=<-0.0208 0.0042 0.0858 -0.9961>}
 middle_02_l {parent=middle_01_l, loc=<4.0123 0.1401 0.1175>, rot=<-0.0218 0.0258 0.3512 -0.9357>}
 middle_03_l {parent=middle_02_l, loc=<2.6077 0.0534 0.1373>, rot=<0.0082 0.0188 0.1708 -0.9851>}
 pinky_metacarpal_l {parent=hand_l, loc=<1.7616 -0.9753 3.3814>, rot=<-0.0397 0.0814 -0.069 -0.9935>}
 pinky_01_l {parent=pinky_metacarpal_l, loc=<7.1709 -0.4831 -0.1094>, rot=<-0.0654 -0.0035 0.2242 -0.9723>}
 pinky_02_l {parent=pinky_01_l, loc=<2.8274 0.0893 -0.1043>, rot=<0.0037 0.0011 0.2854 -0.9584>}
 pinky_03_l {parent=pinky_02_l, loc=<2.2046 0.0307 -0.0078>, rot=<-0.0002 0.0098 0.1445 -0.9895>}
 ring_metacarpal_l {parent=hand_l, loc=<2.2074 -0.3625 1.3638>, rot=<-0.0284 0.0722 -0.0458 -0.9959>}
 ring_01_l {parent=ring_metacarpal_l, loc=<7.0295 -0.6197 -0.1446>, rot=<-0.0489 -0.0252 0.1892 -0.9804>}
 ring_02_l {parent=ring_01_l, loc=<3.6174 0.4647 -0.1027>, rot=<0 0 0.2806 -0.9598>}
 ring_03_l {parent=ring_02_l, loc=<2.4152 0.114 0.0199>, rot=<0 0 0.1189 -0.9929>}
 thumb_01_l {parent=hand_l, loc=<1.3118 -0.3191 -2.459>, rot=<0.6546 -0.4042 0.0664 -0.6354>}
 thumb_02_l {parent=thumb_01_l, loc=<5.9336 -0.0448 0.0401>, rot=<-0.0037 0.0213 0.298 -0.9543>}
 thumb_03_l {parent=thumb_02_l, loc=<2.2223 0.3682 -0.0907>, rot=<0.0118 -0.013 0.0066 -0.9998>}
 deform_wrist_dn_base_l {parent=hand_l, loc=<0 -2.2342 0>, rot=<0 0 0 -1>}
 deform_wrist_dn_l {parent=deform_wrist_dn_base_l, loc=<0 0 0>, rot=<0 0 0 1>}
 deform_wrist_up_base_l {parent=hand_l, loc=<0 2.2682 0>, rot=<0 0 0 -1>}
 deform_wrist_up_l {parent=deform_wrist_up_base_l, loc=<0 0 0>, rot=<0 0 0 1>}
 lowerarm_twist_01_l {parent=lowerarm_l, loc=<17.8549 0 0>, rot=<0 0 0 1>}
 lowerarm_twist_02_l {parent=lowerarm_l, loc=<9.6199 0 0>, rot=<0 0 0 1>}
 deform_elbow_in_base_l {parent=lowerarm_l, loc=<0.2112 5.259 0.8274>, rot=<0 0 0 1>}
 deform_elbow_in_l {parent=deform_elbow_in_base_l, loc=<0 0 0>, rot=<0 0 0 1>}
 deform_elbow_out_base_l {parent=lowerarm_l, loc=<0 -4 0>, rot=<0 0 0 1>}
 deform_elbow_out_l {parent=deform_elbow_out_base_l, loc=<0 0 0>, rot=<0 0 0 1>}
 hand_l_null {parent=lowerarm_l, type=Null, loc=<22.3491 0 0>, rot=<-0.7081 -0.0507 0.0975 -0.6975>}
 upperarm_twist_01_l {parent=upperarm_l, loc=<0.0041 0 0>, rot=<0 0 0 1>}
 deform_upperarm_bk_base_l {parent=upperarm_twist_01_l, loc=<1.0813 -7.2193 -1.9453>, rot=<0 0 0 1>}
 deform_upperarm_bk_l {parent=deform_upperarm_bk_base_l, loc=<0 0 0>, rot=<0 0 0 1>}
 deform_upperarm_fr_base_l {parent=upperarm_twist_01_l, loc=<-0.3691 7.0506 -2.765>, rot=<0 0 0 1>}
 deform_upperarm_fr_l {parent=deform_upperarm_fr_base_l, loc=<0 0 0>, rot=<0 0 0 1>}
 deform_upperarm_up_base_l {parent=upperarm_twist_01_l, loc=<-3.2216 -1.1391 3.3311>, rot=<0 0 0 1>}
 deform_upperarm_up_l {parent=deform_upperarm_up_base_l, loc=<0 0 0>, rot=<0 0 0 1>}
 upperarm_twist_02_l {parent=upperarm_l, loc=<13.7941 0 0>, rot=<0 0 0 1>}
 stk_sTube_a_2_l {parent=upperarm_l, loc=<-7.6085 -0.8212 4.5117>, rot=<0.0955 0.102 -0.7033 -0.6971>}
 stk_sTube_a_3_l {parent=upperarm_l, loc=<-6.8892 -9.7848 -2.7372>, rot=<-0.2751 0.4404 -0.6446 -0.561>}
 stk_sTube_a_1_l {parent=upperarm_l, loc=<-5.429 8.218 -1.9226>, rot=<0.404 -0.2491 -0.5618 -0.6776>}
 stk_sTube_b_3_l {parent=upperarm_l, loc=<-2.6477 -9.9433 -1.5459>, rot=<-0.302 0.4301 -0.6567 -0.5409>}
 stk_sTube_b_2_l {parent=upperarm_l, loc=<-3.6445 -0.6171 5.8128>, rot=<0.1084 0.1264 -0.6881 -0.7062>}
 stk_sTube_b_1_l {parent=upperarm_l, loc=<-1.6168 8.524 -0.9028>, rot=<0.4376 -0.2771 -0.544 -0.6602>}
 stk_sTube_c_3_l {parent=upperarm_l, loc=<1.7257 -9.553 -0.4098>, rot=<-0.3423 0.4171 -0.6694 -0.5106>}
 stk_sTube_c_2_l {parent=upperarm_l, loc=<0.2584 -0.4468 7.2721>, rot=<0.087 0.0803 -0.696 -0.7082>}
 stk_sTube_c_1_l {parent=upperarm_l, loc=<1.9524 8.4726 0.2311>, rot=<0.4237 -0.3178 -0.5099 -0.6778>}
 upperarm_l_null {parent=clavicle_l, type=Null, loc=<15.5111 0 0>, rot=<0.0226 -0.3234 -0.0088 -0.9459>}
 clavicle_r {parent=spine_05, loc=<4.3575 -0.1466 2.3223>, rot=<0.7289 0.0465 -0.6829 0.009>}
 upperarm_r {parent=clavicle_r, loc=<-15.5111 0 0>, rot=<0.0226 -0.3234 -0.0088 -0.9459>}
 lowerarm_r {parent=upperarm_r, loc=<-26.6515 0 0>, rot=<0.0391 -0.0099 -0.2453 -0.9686>}
 hand_r {parent=lowerarm_r, loc=<-22.3492 0 0>, rot=<-0.7081 -0.0507 0.0975 -0.6975>}
 index_metacarpal_r {parent=hand_r, loc=<-1.6534 -0.0181 2.9601>, rot=<0.0158 -0.0124 -0.0173 -0.9996>}
 index_01_r {parent=index_metacarpal_r, loc=<-8.3008 -0.0357 0.0524>, rot=<-0.0044 0.0485 0.0924 -0.9945>}
 index_02_r {parent=index_01_r, loc=<-3.3589 0.3003 -0.0453>, rot=<-0.002 0.0027 0.3225 -0.9466>}
 index_03_r {parent=index_02_r, loc=<-2.5578 0.1231 0.0251>, rot=<-0.002 0.0067 0.2558 -0.9667>}
 middle_metacarpal_r {parent=hand_r, loc=<-2.1856 0.1726 0.5984>, rot=<0.006 0.0168 -0.0329 -0.9993>}
 middle_01_r {parent=middle_metacarpal_r, loc=<-7.3273 0.2447 -0.0129>, rot=<-0.0208 0.0042 0.0858 -0.9961>}
 middle_02_r {parent=middle_01_r, loc=<-4.0142 -0.1411 -0.07>, rot=<-0.0217 0.0259 0.3512 -0.9357>}
 middle_03_r {parent=middle_02_r, loc=<-2.6074 -0.0534 -0.1312>, rot=<0.0083 0.0188 0.1708 -0.9851>}
 pinky_metacarpal_r {parent=hand_r, loc=<-1.7617 0.9753 -3.3814>, rot=<-0.0397 0.0814 -0.069 -0.9935>}
 pinky_01_r {parent=pinky_metacarpal_r, loc=<-7.1707 0.4832 0.1094>, rot=<-0.0654 -0.0035 0.2242 -0.9723>}
 pinky_02_r {parent=pinky_01_r, loc=<-2.8274 -0.0892 0.1043>, rot=<0.0037 0.0011 0.2854 -0.9584>}
 pinky_03_r {parent=pinky_02_r, loc=<-2.2045 -0.031 0.0078>, rot=<-0.0002 0.0098 0.1445 -0.9895>}
 ring_metacarpal_r {parent=hand_r, loc=<-2.2073 0.3627 -1.3638>, rot=<-0.0284 0.0722 -0.0458 -0.9959>}
 ring_01_r {parent=ring_metacarpal_r, loc=<-7.0297 0.6197 0.1446>, rot=<-0.0489 -0.0252 0.1892 -0.9804>}
 ring_02_r {parent=ring_01_r, loc=<-3.6171 -0.4648 0.1027>, rot=<0 0 0.2806 -0.9598>}
 ring_03_r {parent=ring_02_r, loc=<-2.4153 -0.1141 -0.0199>, rot=<0 0 0.1189 -0.9929>}
 thumb_01_r {parent=hand_r, loc=<-1.3119 0.3192 2.459>, rot=<0.6546 -0.4042 0.0664 -0.6354>}
 thumb_02_r {parent=thumb_01_r, loc=<-5.9335 0.0447 -0.0401>, rot=<-0.0037 0.0213 0.298 -0.9543>}
 thumb_03_r {parent=thumb_02_r, loc=<-2.2223 -0.3682 0.0907>, rot=<0.0118 -0.013 0.0066 -0.9998>}
 deform_wrist_dn_base_r {parent=hand_r, loc=<-0.0003 2.2346 -0.0001>, rot=<0 0 0 -1>}
 deform_wrist_dn_r {parent=deform_wrist_dn_base_r, loc=<0 0 0>, rot=<0 0 0 1>}
 deform_wrist_up_base_r {parent=hand_r, loc=<0.0004 -2.2683 0.0001>, rot=<0 0 0 -1>}
 deform_wrist_up_r {parent=deform_wrist_up_base_r, loc=<0 0 0>, rot=<0 0 0 1>}
 lowerarm_twist_01_r {parent=lowerarm_r, loc=<-17.855 0 0>, rot=<0 0 0 1>}
 lowerarm_twist_02_r {parent=lowerarm_r, loc=<-9.6199 0 0>, rot=<0 0 0 1>}
 deform_elbow_in_base_r {parent=lowerarm_r, loc=<-0.2111 -5.2591 -0.8275>, rot=<0 0 0 1>}
 deform_elbow_in_r {parent=deform_elbow_in_base_r, loc=<-0.0001 0 0.0002>, rot=<0 0 0 1>}
 deform_elbow_out_base_r {parent=lowerarm_r, loc=<0 4 0>, rot=<0 0 0 1>}
 deform_elbow_out_r {parent=deform_elbow_out_base_r, loc=<0 0 0>, rot=<0 0 0 1>}
 hand_r_null {parent=lowerarm_r, type=Null, loc=<-22.3492 0 0>, rot=<-0.7081 -0.0507 0.0975 -0.6975>}
 upperarm_twist_01_r {parent=upperarm_r, loc=<-0.0041 0 0>, rot=<0 0 0 1>}
 deform_upperarm_bk_base_r {parent=upperarm_twist_01_r, loc=<-1.0815 7.2192 1.9456>, rot=<0 0 0 1>}
 deform_upperarm_bk_r {parent=deform_upperarm_bk_base_r, loc=<0 0 0>, rot=<0 0 0 1>}
 deform_upperarm_fr_base_r {parent=upperarm_twist_01_r, loc=<0.3689 -7.0506 2.7652>, rot=<0 0 0 1>}
 deform_upperarm_fr_r {parent=deform_upperarm_fr_base_r, loc=<0 0 0>, rot=<0 0 0 1>}
 deform_upperarm_up_base_r {parent=upperarm_twist_01_r, loc=<3.2213 1.1391 -3.3308>, rot=<0 0 0 1>}
 deform_upperarm_up_r {parent=deform_upperarm_up_base_r, loc=<0 0 0>, rot=<0 0 0 -1>}
 upperarm_twist_02_r {parent=upperarm_r, loc=<-13.7942 0 0>, rot=<0 0 0 1>}
 stk_sTube_a_2_r {parent=upperarm_r, loc=<7.6087 0.8212 -4.5119>, rot=<0.7033 -0.6971 0.0955 -0.102>}
 stk_sTube_a_3_r {parent=upperarm_r, loc=<6.889 9.7848 2.7375>, rot=<0.6446 -0.561 -0.2751 -0.4404>}
 stk_sTube_a_1_r {parent=upperarm_r, loc=<5.4289 -8.218 1.9228>, rot=<0.5618 -0.6776 0.404 0.2491>}
 stk_sTube_b_3_r {parent=upperarm_r, loc=<2.6475 9.9432 1.5463>, rot=<0.6567 -0.5409 -0.302 -0.4301>}
 stk_sTube_b_2_r {parent=upperarm_r, loc=<3.6442 0.6172 -5.8125>, rot=<0.6881 -0.7062 0.1084 -0.1264>}
 stk_sTube_b_1_r {parent=upperarm_r, loc=<1.6168 -8.524 0.9029>, rot=<0.544 -0.6602 0.4376 0.2771>}
 stk_sTube_c_3_r {parent=upperarm_r, loc=<-1.7254 9.5531 0.4094>, rot=<0.6694 -0.5106 -0.3423 -0.4171>}
 stk_sTube_c_2_r {parent=upperarm_r, loc=<-0.2584 0.4468 -7.272>, rot=<0.696 -0.7082 0.087 -0.0803>}
 stk_sTube_c_1_r {parent=upperarm_r, loc=<-1.9523 -8.4725 -0.2311>, rot=<0.5099 -0.6778 0.4237 0.3178>}
 upperarm_r_null {parent=clavicle_r, type=Null, loc=<-15.5111 0 0>, rot=<0.0226 -0.3234 -0.0088 -0.9459>}
 neck_01 {parent=spine_05, loc=<11.3138 0.1529 0>, rot=<0 0 -0.2003 -0.9797>}
 neck_02 {parent=neck_01, loc=<4.5627 0 0>, rot=<0 0 0.0504 -0.9987>}
 head {parent=neck_02, loc=<5.3914 0 0>, rot=<0 0 0.115 -0.9934>}
 base_hood_c {parent=spine_05, loc=<11.8792 -8.4524 0>, rot=<-0.6981 -0.1125 -0.1125 -0.6981>}
 dyn_hood_c {parent=base_hood_c, loc=<0 0 0>, rot=<0 0 0 1>}
 base_hood_3_l {parent=spine_05, loc=<13.6393 -6.1098 -8.3697>, rot=<0.9177 0.1611 0.1603 0.3259>}
 dyn_hood_3_l {parent=base_hood_3_l, loc=<0 0 0>, rot=<0 0 0 1>}
 base_hood_2_l {parent=spine_05, loc=<12.0795 1.1835 -10.608>, rot=<0.9807 0.1816 0.0543 -0.0483>}
 dyn_hood_2_l {parent=base_hood_2_l, loc=<0 0 0>, rot=<0 0 0 1>}
 base_hood_1_l {parent=spine_05, loc=<7.7837 7.0936 -8.8481>, rot=<0.9668 0.187 0.031 -0.1714>}
 dyn_hood_1_l {parent=base_hood_1_l, loc=<0 0 0>, rot=<0 0 0 1>}
 base_hood_3_r {parent=spine_05, loc=<13.6395 -6.1099 8.3697>, rot=<-0.3259 -0.1603 -0.1611 -0.9177>}
 dyn_hood_3_r {parent=base_hood_3_r, loc=<0 0 0>, rot=<0 0 0 1>}
 base_hood_2_r {parent=spine_05, loc=<12.08 1.1836 10.608>, rot=<0.0483 -0.0543 -0.1816 -0.9807>}
 dyn_hood_2_r {parent=base_hood_2_r, loc=<0 0 0>, rot=<0 0 0 1>}
 base_hood_1_r {parent=spine_05, loc=<7.7836 7.0936 8.8481>, rot=<0.1714 -0.031 -0.187 -0.9668>}
 dyn_hood_1_r {parent=base_hood_1_r, loc=<0 0 0>, rot=<0 0 0 1>}
 head_null {parent=spine_05, type=Null, loc=<20.6573 3.5489 0>, rot=<0 0 -0.036 -0.9994>}
 clavicle_r_null {parent=spine_05, type=Null, loc=<4.3575 -0.1466 2.3223>, rot=<0.7289 0.0465 -0.6829 0.009>}
 clavicle_l_null {parent=spine_05, type=Null, loc=<4.3575 -0.1466 -2.3223>, rot=<0.0465 -0.7289 -0.009 -0.6829>}
 base_internalNeckCollar_1_r {parent=spine_05, loc=<6.9799 6.4458 5.6267>, rot=<0.7906 0.2131 -0.2292 -0.5264>}
 dyn_internalNeckCollar_1_r {parent=base_internalNeckCollar_1_r, loc=<0 0 0>, rot=<0 0 0 1>}
 base_internalNeckCollar_1_c {parent=spine_05, loc=<1.5125 10.2094 0>, rot=<0.9635 0.2677 0 0>}
 dyn_internalNeckCollar_1_c {parent=base_internalNeckCollar_1_c, loc=<0 0 0>, rot=<0 0 0 1>}
 base_internalNeckCollar_1_l {parent=spine_05, loc=<6.9803 6.4458 -5.6267>, rot=<-0.7906 -0.2131 -0.2292 -0.5264>}
 dyn_internalNeckCollar_1_l {parent=base_internalNeckCollar_1_l, loc=<0 0 0>, rot=<0 0 0 1>}
 deform_pec_aim_base_l {parent=spine_04, loc=<7.8732 12.6388 -2.4123>, rot=<0.0102 -0.4602 0.3979 -0.7936>}
 deform_pec_aim_l {parent=deform_pec_aim_base_l, loc=<0 0 0>, rot=<0 0 0 1>}
 deform_pec_base_l {parent=deform_pec_aim_l, loc=<10.766 0 0>, rot=<0 0 0 1>}
 deform_pec_l {parent=deform_pec_base_l, loc=<0.055 0 0>, rot=<0 0 0 1>}
 deform_pec_aim_base_r {parent=spine_04, loc=<7.8727 12.6388 2.4123>, rot=<-0.4602 -0.0102 0.7936 0.3979>}
 deform_pec_aim_r {parent=deform_pec_aim_base_r, loc=<0 0 0>, rot=<0 0 0 1>}
 deform_pec_base_r {parent=deform_pec_aim_r, loc=<-10.7662 -0.0006 -0.0007>, rot=<0 0 0 1>}
 deform_pec_r {parent=deform_pec_base_r, loc=<0 0 0>, rot=<0 0 0 1>}
 deform_lat_l_base {parent=spine_04, loc=<0.5455 0.6179 -13.778>, rot=<0 0 0 1>}
 deform_lat_l {parent=deform_lat_l_base, loc=<0 0 0>, rot=<0 0 0 1>}
 deform_lat_r_base {parent=spine_04, loc=<0.5455 0.6179 13.778>, rot=<0 0 0 1>}
 deform_lat_r {parent=deform_lat_r_base, loc=<0 0 0>, rot=<0 0 0 1>}
 base_tailCoat_back_c_1 {parent=pelvis, loc=<10.6257 -9.3606 0>, rot=<0.0827 -0.7023 0.7023 -0.0827>}
 dyn_tailCoat_back_c_1 {parent=base_tailCoat_back_c_1, loc=<0 0 0>, rot=<0 0 0 1>}
 dyn_tailCoat_back_c_2 {parent=dyn_tailCoat_back_c_1, loc=<10.0385 0 0>, rot=<0 -0.0254 0 -0.9997>}
 thigh_l {parent=pelvis, loc=<-2.5693 -0.0109 -7.7501>, rot=<0.0012 0.0572 0.0155 -0.9982>}
 calf_l {parent=thigh_l, loc=<-39.0484 0 0>, rot=<0.0006 -0.0161 -0.0391 -0.9991>}
 calf_twist_01_l {parent=calf_l, loc=<-35.813 -0.0967 0.0397>, rot=<0.0237 0.0095 -0.0035 -0.9997>}
 calf_twist_02_l {parent=calf_l, loc=<-18.8064 -0.0967 0.0397>, rot=<0.0237 0.0095 -0.0035 -0.9997>}
 foot_l {parent=calf_l, loc=<-42.561 0.1077 0.4292>, rot=<0.0418 -0.0455 0.0343 -0.9975>}
 ball_l {parent=foot_l, loc=<-6.1599 13.8264 0.6806>, rot=<0.0001 -0.0001 -0.7063 -0.7079>}
 deform_knee_in_base_l {parent=calf_l, loc=<1.298 -5.2957 -0.0496>, rot=<0 0 0 1>}
 deform_knee_in_l {parent=deform_knee_in_base_l, loc=<0 0 0>, rot=<0 0 0 1>}
 deform_knee_out_base_l {parent=calf_l, loc=<0 4 0>, rot=<0 0 0 1>}
 deform_knee_out_l {parent=deform_knee_out_base_l, loc=<0 0 0>, rot=<0 0 0 1>}
 foot_l_null {parent=calf_l, type=Null, loc=<-42.561 0.1077 0.4292>, rot=<0.0418 -0.0455 0.0343 -0.9975>}
 calf_twist_00_l {parent=calf_l, loc=<0 0 0>, rot=<0 0 0 1>}
 thigh_twist_01_l {parent=thigh_l, loc=<-21.6303 0 0>, rot=<0.0013 0 0 -1>}
 deform_groin_base_l {parent=thigh_l, loc=<-2.1908 8.1142 -1.7442>, rot=<0 0 0 -1>}
 deform_groin_l {parent=deform_groin_base_l, loc=<0 0 0>, rot=<0 0 0 1>}
 calf_l_null {parent=thigh_l, type=Null, loc=<-39.0484 0 0>, rot=<0.0006 -0.0161 -0.0391 -0.9991>}
 thigh_r {parent=pelvis, loc=<-2.5691 -0.0109 7.7501>, rot=<0.0572 -0.0012 0.9982 0.0155>}
 calf_r {parent=thigh_r, loc=<39.0487 0 0>, rot=<0.0006 -0.0161 -0.0391 -0.9991>}
 calf_twist_01_r {parent=calf_r, loc=<35.8132 0.0967 -0.0397>, rot=<0.0237 0.0095 -0.0035 -0.9997>}
 calf_twist_02_r {parent=calf_r, loc=<18.8064 0.0967 -0.0397>, rot=<0.0237 0.0095 -0.0035 -0.9997>}
 foot_r {parent=calf_r, loc=<42.5612 -0.1077 -0.4292>, rot=<0.0418 -0.0455 0.0343 -0.9975>}
 ball_r {parent=foot_r, loc=<6.1598 -13.8264 -0.6806>, rot=<0.0001 -0.0001 -0.7063 -0.7079>}
 deform_knee_in_base_r {parent=calf_r, loc=<-1.2981 5.2958 0.0496>, rot=<0 0 0 1>}
 deform_knee_in_r {parent=deform_knee_in_base_r, loc=<0 0 0>, rot=<0 0 0 1>}
 deform_knee_out_base_r {parent=calf_r, loc=<0 -4 0>, rot=<0 0 0 1>}
 deform_knee_out_r {parent=deform_knee_out_base_r, loc=<0 0 0>, rot=<0 0 0 1>}
 foot_r_null {parent=calf_r, type=Null, loc=<42.5612 -0.1077 -0.4292>, rot=<0.0418 -0.0455 0.0343 -0.9975>}
 calf_twist_00_r {parent=calf_r, loc=<0 0 0>, rot=<0 0 0 1>}
 thigh_twist_01_r {parent=thigh_r, loc=<21.6303 0 0>, rot=<0.0013 0 0 -1>}
 deform_groin_base_r {parent=thigh_r, loc=<2.1909 -8.1142 1.7442>, rot=<0 0 0 -1>}
 deform_groin_r {parent=deform_groin_base_r, loc=<0 0 0>, rot=<0 0 0 1>}
 calf_r_null {parent=thigh_r, type=Null, loc=<39.0487 0 0>, rot=<0.0006 -0.0161 -0.0391 -0.9991>}
 deform_glute_base_l {parent=pelvis, loc=<-8.8064 -8.8222 -8.4287>, rot=<0.0275 0.0501 -0.4472 -0.8926>}
 deform_glute_l {parent=deform_glute_base_l, loc=<0 0 0>, rot=<0 0 0 1>}
 deform_glute_base_r {parent=pelvis, loc=<-8.8062 -8.8222 8.4287>, rot=<0.0501 -0.0275 0.8926 -0.4472>}
 deform_glute_r {parent=deform_glute_base_r, loc=<0 0 0>, rot=<0 0 0 1>}
 base_radio {parent=pelvis, loc=<0.6303 0.3988 17.3823>, rot=<-0.0333 0.6669 -0.0268 -0.7439>}
 dyn_radio {parent=base_radio, loc=<0 0 0>, rot=<0 0 0 1>}
 base_radioWire {parent=dyn_radio, loc=<0.1027 0.0695 2.8718>, rot=<-0.006 0.6831 0.0118 -0.7302>}
 dyn_radioWire {parent=base_radioWire, loc=<0 0 0>, rot=<0 0 0 1>}
 dyn_radioWire_m {parent=dyn_radioWire, loc=<9.8846 0 0>, rot=<0 -0.0236 0 -0.9997>}
 base_radioAntenna {parent=dyn_radio, loc=<0.1364 -3.5962 -4.742>, rot=<0.0127 -0.7013 0.011 -0.7127>}
 dyn_radioAntenna {parent=base_radioAntenna, loc=<0 0 0>, rot=<0 0 0 1>}
 thigh_l_null {parent=pelvis, type=Null, loc=<-2.5693 -0.0109 -7.7501>, rot=<0.0012 0.0572 0.0155 -0.9982>}
 thigh_r_null {parent=pelvis, type=Null, loc=<-2.5691 -0.0109 7.7501>, rot=<0.0572 -0.0012 0.9982 0.0155>}
 baseW_tailCoat_front_l_1 {parent=pelvis, loc=<10.9405 7.3477 -10.1519>, rot=<0 -0.7071 -0.7071 0>}
 base_tailCoat_front_l_1 {parent=baseW_tailCoat_front_l_1, loc=<0 0 0>, rot=<-0.2547 0.1586 -0.0212 -0.9537>}
 dyn_tailCoat_front_l_1 {parent=base_tailCoat_front_l_1, loc=<0 0 0>, rot=<0 0 0 1>}
 dyn_tailCoat_front_l_2 {parent=dyn_tailCoat_front_l_1, loc=<9.7487 0 0>, rot=<0 -0.0101 -0.0029 -0.9999>}
 baseW_tailCoat_front_r_1 {parent=pelvis, loc=<10.9401 7.3477 10.1519>, rot=<0 -0.7071 -0.7071 0>}
 base_tailCoat_front_r_1 {parent=baseW_tailCoat_front_r_1, loc=<0 0 0>, rot=<0.2547 0.1586 0.0212 -0.9537>}
 dyn_tailCoat_front_r_1 {parent=base_tailCoat_front_r_1, loc=<0 0 0>, rot=<0 0 0 1>}
 dyn_tailCoat_front_r_2 {parent=dyn_tailCoat_front_r_1, loc=<9.7484 0 0.0001>, rot=<0 -0.0101 0.0029 -0.9999>}
 pelvis_ctrl {controlType=EULER_TRANSFORM, type=Control, loc=<0 0 0>, rot=<0 0 0 1>}
 spine_01_ctrl {controlType=EULER_TRANSFORM, parent=pelvis_ctrl, type=Control, loc=<0 0 0>, rot=<0 0 0 1>}
 spine_02_ctrl {controlType=EULER_TRANSFORM, parent=spine_01_ctrl, type=Control, loc=<0 0 0>, rot=<0 0 0 1>}
 spine_03_ctrl {controlType=EULER_TRANSFORM, parent=spine_02_ctrl, type=Control, loc=<0 0 0>, rot=<0 0 0 1>}
 spine_04_ctrl {controlType=EULER_TRANSFORM, parent=spine_03_ctrl, type=Control, loc=<0 0 0>, rot=<0 0 0 1>}
 spine_05_ctrl {controlType=EULER_TRANSFORM, parent=spine_04_ctrl, type=Control, loc=<0 0 0>, rot=<0 0 0 1>}
 clavicle_l_ctrl {controlType=EULER_TRANSFORM, parent=spine_05_ctrl, type=Control, loc=<0 0 0>, rot=<0 0 0 1>}
 upperarm_l_ctrl {controlType=EULER_TRANSFORM, parent=clavicle_l_ctrl, type=Control, loc=<0 0 0>, rot=<0 0 0 1>}
 lowerarm_l_ctrl {controlType=EULER_TRANSFORM, parent=upperarm_l_ctrl, type=Control, loc=<0 0 0>, rot=<0 0 0 1>}
 hand_l_ctrl {controlType=EULER_TRANSFORM, parent=lowerarm_l_ctrl, type=Control, loc=<0 0 0>, rot=<0 0 0 1>}
 clavicle_r_ctrl {controlType=EULER_TRANSFORM, parent=spine_05_ctrl, type=Control, loc=<0 0 0>, rot=<0 0 0 1>}
 upperarm_r_ctrl {controlType=EULER_TRANSFORM, parent=clavicle_r_ctrl, type=Control, loc=<0 0 0>, rot=<0 0 0 1>}
 lowerarm_r_ctrl {controlType=EULER_TRANSFORM, parent=upperarm_r_ctrl, type=Control, loc=<0 0 0>, rot=<0 0 0 1>}
 hand_r_ctrl {controlType=EULER_TRANSFORM, parent=lowerarm_r_ctrl, type=Control, loc=<0 0 0>, rot=<0 0 0 1>}
 neck_01_ctrl {controlType=EULER_TRANSFORM, parent=spine_05_ctrl, type=Control, loc=<0 0 0>, rot=<0 0 0 1>}
 neck_02_ctrl {controlType=EULER_TRANSFORM, parent=neck_01_ctrl, type=Control, loc=<0 0 0>, rot=<0 0 0 1>}
 head_ctrl {controlType=EULER_TRANSFORM, parent=neck_02_ctrl, type=Control, loc=<0 0 0>, rot=<0 0 0 1>}
 thigh_l_ctrl {controlType=EULER_TRANSFORM, parent=pelvis_ctrl, type=Control, loc=<0 0 0>, rot=<0 0 0 1>}
 calf_l_ctrl {controlType=EULER_TRANSFORM, parent=thigh_l_ctrl, type=Control, loc=<0 0 0>, rot=<0 0 0 1>}
 foot_l_ctrl {controlType=EULER_TRANSFORM, parent=calf_l_ctrl, type=Control, loc=<0 0 0>, rot=<0 0 0 1>}
 thigh_r_ctrl {controlType=EULER_TRANSFORM, parent=pelvis_ctrl, type=Control, loc=<0 0 0>, rot=<0 0 0 1>}
 calf_r_ctrl {controlType=EULER_TRANSFORM, parent=thigh_r_ctrl, type=Control, loc=<0 0 0>, rot=<0 0 0 1>}
 foot_r_ctrl {controlType=EULER_TRANSFORM, parent=calf_r_ctrl, type=Control, loc=<0 0 0>, rot=<0 0 0 1>}
]
