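# Physics asset for the M_MED_FitVapor male player body, set up for secondary motion on its tailcoat flaps and radio with wire.
# Back, front-left and front-right tails are two-segment chains from pelvis-linked base bodies, with limited swing and little or no twist.
# The radio links to the pelvis with free rotation and carries a limited two-segment wire chain through its own base body.
# Capsules on pelvis, lower spine, thighs and feet supply body shapes.
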PhysicsAsset /BRCosmetics/Characters/Player/Male/Medium/Bodies/M_MED_FitVapor/Meshes/M_MED_FitVapor_Physics
bodies {
 base_radioWire [{shapeType=Capsule, length=2, radius=1, shapeName=base_radioWire_capsule}]
 base_tailCoat_back_c_1 [{shapeType=Capsule, length=0.505, radius=0.505, shapeName=base_tailCoat_back_c_1_capsule, rotation=<rot 0 -90 0>}]
 base_tailCoat_front_l_1 [{shapeType=Capsule, length=2, radius=1, shapeName=base_tailCoat_front_l_1_capsule}]
 base_tailCoat_front_r_1 [{shapeType=Capsule, length=2, radius=1, shapeName=base_tailCoat_front_l_1_capsule, rotation=<rot 0 0 -180>}]
 dyn_radio [
  {shapeType=Capsule, length=7.0264, radius=2.379, shapeName=dyn_radio_capsuleA, center=<0.0621 -0.0989 -1.7126>, rotation=<rot 0 0 90>}
  {shapeType=Capsule, length=7.0264, radius=2.379, shapeName=dyn_radio_capsuleB, center=<0.0621 -0.0989 1.0955>, rotation=<rot 0 0 90>}
 ]
 dyn_radioWire [
  {
   shapeType=Capsule
   length=8.25
   radius=1.9
   shapeName=dyn_radioWire_capsuleA
   center=<4.8724 -6.8634 0>
   rotation=<rot 90 179.1048 -171.1445>
  }
  {
   shapeType=Capsule
   length=8.2445
   radius=1.6161
   shapeName=dyn_radioWire_capsuleB
   center=<3.7404 6.928 0.2538>
   rotation=<rot 90 180 161.6413>
  }
 ]
 dyn_radioWire_m [
  {
   shapeType=Capsule
   length=9.6842
   radius=1.4924
   shapeName=dyn_radioWire_m_capsuleA
   center=<5.8035 -4.5201 0.4603>
   rotation=<rot -0.0001 122 -90.0001>
  }
  {
   shapeType=Capsule
   length=9.6842
   radius=1.4924
   shapeName=dyn_radioWire_m_capsuleB
   center=<4.4728 6.2038 0.4602>
   rotation=<rot -0.0002 65.8064 -90>
  }
  {
   shapeType=Capsule
   length=5.8692
   radius=1.1847
   shapeName=dyn_radioWire_m_capsuleC
   center=<10.3061 0.8157 0.554>
   rotation=<rot -0.0001 1.4561 -89.9998>
  }
 ]
 dyn_tailCoat_back_c_1 [{shapeType=Box, extentX=8.8365, extentY=21.8805, extentZ=2.351, shapeName=dyn_tailCoat_back_c_2_box, center=<5.72 0 0.4945>}]
 dyn_tailCoat_back_c_2 [{shapeType=Box, extentX=13.8725, extentY=21.2205, extentZ=2.351, shapeName=dyn_tailCoat_back_c_2_box, center=<5.72 0 0.9406>}]
 dyn_tailCoat_front_l_1 [{
   shapeType=Box
   extentX=8.6781
   extentY=13.311
   extentZ=2.6335
   shapeName=dyn_tailCoat_front_l_1_box
   center=<3.7221 0 0.1933>
   rotation=<rot -0.2191 -1.9879 -6.2906>
  }]
 dyn_tailCoat_front_l_2 [{
   shapeType=Box
   extentX=11.3952
   extentY=12.4908
   extentZ=1.7649
   shapeName=dyn_tailCoat_front_l_2_box
   center=<4.53 -0.5752 -0.0158>
   rotation=<rot 0.7189 -3.2969 -3.4884>
  }]
 dyn_tailCoat_front_r_1 [
  {
   shapeType=Capsule
   length=9.3586
   radius=1.2712
   shapeName=dyn_tailCoat_front_r_1_capsule
   center=<5.4714 5.8483 -1.3294>
   rotation=<rot -81.0515 2.007 -8.0609>
  }
  {
   shapeType=Box
   extentX=11.3952
   extentY=12.4908
   extentZ=1.7649
   shapeName=dyn_tailCoat_front_l_2_box
   center=<4.53 0.5752 0.0158>
   rotation=<rot 0.7189 -3.2969 176.5116>
  }
 ]
 dyn_tailCoat_front_r_2 [{
   shapeType=Box
   extentX=11.3952
   extentY=12.4908
   extentZ=1.7649
   shapeName=dyn_tailCoat_front_l_2_box
   center=<4.53 0.5752 0.0158>
   rotation=<rot 0.7189 -3.2969 176.5116>
  }]
 foot_l [{
   shapeType=Capsule
   length=10.8809
   radius=6.6854
   shapeName=foot_r_capsule
   center=<-4.9238 4.8089 -0.1098>
   rotation=<rot 0 -176.6726 92.8318>
  }]
 foot_r [{
   shapeType=Capsule
   length=10.8809
   radius=6.6854
   shapeName=foot_r_capsule
   center=<4.9238 -4.8089 0.1098>
   rotation=<rot 0 -176.6726 -87.1682>
  }]
 pelvis [{
   shapeType=Capsule
   length=9.7166
   radius=12.1824
   shapeName=pelvis_capsule
   center=<-3.571 -0.0435 -0.1174>
   rotation=<rot -0.2743 -90 -1.1665>
  }]
 spine_01 [{shapeType=Capsule, length=15.5527, radius=8.6197, shapeName=pelvis_capsule, rotation=<rot -0.2743 -90 -1.1665>}]
 thigh_l [
  {
   shapeType=Capsule
   length=32.003
   radius=10.4835
   shapeName=thigh_r_capsule
   center=<-17.8855 0.5426 0.6578>
   rotation=<rot -6.9048 -90.0005 83.8564>
  }
  {
   shapeType=Capsule
   length=12.0709
   radius=3.5263
   shapeName=thigh_l_capsule_plus
   center=<-15.6879 8.1284 -0.441>
   rotation=<rot 12.3809 -0.0004 7.7591>
  }
 ]
 thigh_r [
  {
   shapeType=Capsule
   length=32.003
   radius=10.4835
   shapeName=thigh_r_capsule
   center=<17.8855 -0.5426 -0.6578>
   rotation=<rot -6.9048 -90.0005 -96.1436>
  }
  {
   shapeType=Capsule
   length=20.0739
   radius=8.7012
   shapeName=thigh_r_capsuleBulge
   center=<18.0776 -2.0403 1.0591>
   rotation=<rot -6.4432 -85.8504 -96.6262>
  }
 ]
}
constraints [
 {
  bone1Name=base_tailCoat_back_c_1
  bone2Name=pelvis
  swing1LimitDegrees=45
  swing1Motion=Free
  swing2LimitDegrees=45
  swing2Motion=Free
  twistLimitDegrees=45
  twistMotion=Free
 }
 {bone1Name=dyn_tailCoat_back_c_1, bone2Name=base_tailCoat_back_c_1, swing1LimitDegrees=15, swing2LimitDegrees=20, twistLimitDegrees=5}
 {bone1Name=dyn_tailCoat_back_c_2, bone2Name=dyn_tailCoat_back_c_1, swing1LimitDegrees=20, swing2LimitDegrees=25, twistLimitDegrees=5}
 {
  bone1Name=base_tailCoat_front_l_1
  bone2Name=pelvis
  swing1LimitDegrees=45
  swing1Motion=Free
  swing2LimitDegrees=45
  swing2Motion=Free
  twistLimitDegrees=45
  twistMotion=Free
 }
 {bone1Name=dyn_tailCoat_front_l_1, bone2Name=base_tailCoat_front_l_1, swing1LimitDegrees=20, swing2LimitDegrees=20, twistLimitDegrees=0}
 {bone1Name=dyn_tailCoat_front_l_2, bone2Name=dyn_tailCoat_front_l_1, swing1LimitDegrees=25, swing2LimitDegrees=25, twistLimitDegrees=0}
 {
  bone1Name=thigh_l
  bone2Name=pelvis
  swing1LimitDegrees=45
  swing1Motion=Free
  swing2LimitDegrees=45
  swing2Motion=Free
  twistLimitDegrees=45
  twistMotion=Free
 }
 {
  bone1Name=thigh_r
  bone2Name=pelvis
  swing1LimitDegrees=45
  swing1Motion=Free
  swing2LimitDegrees=45
  swing2Motion=Free
  twistLimitDegrees=45
  twistMotion=Free
 }
 {
  bone1Name=spine_01
  bone2Name=pelvis
  swing1LimitDegrees=45
  swing1Motion=Free
  swing2LimitDegrees=45
  swing2Motion=Free
  twistLimitDegrees=45
  twistMotion=Free
 }
 {
  bone1Name=base_tailCoat_front_r_1
  bone2Name=pelvis
  swing1LimitDegrees=45
  swing1Motion=Free
  swing2LimitDegrees=45
  swing2Motion=Free
  twistLimitDegrees=45
  twistMotion=Free
 }
 {bone1Name=dyn_tailCoat_front_r_1, bone2Name=base_tailCoat_front_r_1, swing1LimitDegrees=20, swing2LimitDegrees=20, twistLimitDegrees=0}
 {bone1Name=dyn_tailCoat_front_r_2, bone2Name=dyn_tailCoat_front_r_1, swing1LimitDegrees=25, swing2LimitDegrees=25, twistLimitDegrees=0}
 {bone1Name=dyn_radioWire, bone2Name=base_radioWire, swing1LimitDegrees=35, swing2LimitDegrees=35, twistLimitDegrees=15}
 {
  bone1Name=base_radioWire
  bone2Name=dyn_radio
  swing1LimitDegrees=45
  swing1Motion=Free
  swing2LimitDegrees=45
  swing2Motion=Free
  twistLimitDegrees=45
  twistMotion=Free
 }
 {
  bone1Name=foot_l
  bone2Name=thigh_l
  swing1LimitDegrees=45
  swing1Motion=Free
  swing2LimitDegrees=45
  swing2Motion=Free
  twistLimitDegrees=45
  twistMotion=Free
 }
 {
  bone1Name=foot_r
  bone2Name=thigh_r
  swing1LimitDegrees=45
  swing1Motion=Free
  swing2LimitDegrees=45
  swing2Motion=Free
  twistLimitDegrees=45
  twistMotion=Free
 }
 {bone1Name=dyn_radioWire_m, bone2Name=dyn_radioWire, swing1LimitDegrees=25, swing2LimitDegrees=30, twistLimitDegrees=10}
 {
  bone1Name=dyn_radio
  bone2Name=pelvis
  swing1LimitDegrees=45
  swing1Motion=Free
  swing2LimitDegrees=45
  swing2Motion=Free
  twistLimitDegrees=45
  twistMotion=Free
 }
]
