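# Physics asset for the hood of the male medium FitVapor cosmetic body, setting up secondary motion of the hood around the head and neck.
# Seven dyn_hood capsules (three per side plus center) are each constrained to a small base_hood anchor capsule, with narrow, asymmetric swing limits and zero-degree twist limits.
# Head, jaw, ear and neck capsules supply body shapes, linked by freely rotating head and neck constraints.
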PhysicsAsset /BRCosmetics/Characters/Player/Male/Medium/Bodies/M_MED_FitVapor/Meshes/M_MED_FitVapor_Physics_hood
bodies {
 base_hood_1_l [{shapeType=Capsule, length=1, radius=0.5, shapeName=base_hood_1_l_capsule}]
 base_hood_1_r [{shapeType=Capsule, length=1, radius=0.5, shapeName=base_hood_1_r_capsule}]
 base_hood_2_l [{shapeType=Capsule, length=1, radius=0.5, shapeName=base_hood_2_l_capsule}]
 base_hood_2_r [{shapeType=Capsule, length=1, radius=0.5, shapeName=base_hood_2_r_capsule}]
 base_hood_3_l [{shapeType=Capsule, length=1, radius=0.5, shapeName=base_hood_3_l_capsule}]
 base_hood_3_r [{shapeType=Capsule, length=1, radius=0.5, shapeName=base_hood_3_r_capsule}]
 base_hood_c [{shapeType=Capsule, length=1, radius=0.5, shapeName=base_hood_c_capsule}]
 dyn_hood_1_l [{shapeType=Capsule, length=3.1595, radius=4.0455, shapeName=dyn_hood_1_l_capsule, center=<1.82 0 2.6668>, rotation=<rot 90 0 0>}]
 dyn_hood_1_r [{shapeType=Capsule, length=3.1595, radius=4.0455, shapeName=dyn_hood_1_l_capsule, center=<1.82 0 2.6668>, rotation=<rot 90 0 180>}]
 dyn_hood_2_l [{
   shapeType=Capsule
   length=5.6448
   radius=3.8711
   shapeName=dyn_hood_1_l_capsule
   center=<2.6274 0 1.8939>
   rotation=<rot 90 153.4349 -116.5651>
  }]
 dyn_hood_2_r [{shapeType=Capsule, length=5.6448, radius=3.8711, shapeName=dyn_hood_1_l_capsule, center=<2.6274 0 1.8939>, rotation=<rot 90 180 90>}]
 dyn_hood_3_l [{shapeType=Capsule, length=4.52, radius=3.6291, shapeName=dyn_hood_1_l_capsule, center=<4.3911 -0.2259 1.0924>, rotation=<rot 90 -90 12>}]
 dyn_hood_3_r [{shapeType=Capsule, length=4.52, radius=3.6291, shapeName=dyn_hood_1_l_capsule, center=<4.3911 0.2259 1.0924>, rotation=<rot 90 -90 -12>}]
 dyn_hood_c [{
   shapeType=Capsule
   length=6.9746
   radius=3.8167
   shapeName=dyn_hood_1_l_capsule
   center=<7.3079 0 1.2069>
   rotation=<rot 90 -165.9638 104.0362>
  }]
 head [
  {shapeType=Capsule, length=6.4924, radius=8.5076, shapeName=head_capsule, center=<9.6947 -1.3616 0>, rotation=<rot -90 180 170.6826>}
  {shapeType=Capsule, length=5.7126, radius=5.0657, shapeName=jaw_capsule, center=<-0.8335 6.9398 0>, rotation=<rot -90 180 170.6826>}
  {shapeType=Capsule, length=7.7909, radius=5.3499, shapeName=ears_capsule, center=<3.4325 0.875 0>}
 ]
 neck_01 [{shapeType=Capsule, length=1, radius=5.1544, shapeName=neck_01_capsule, rotation=<rot -90 0 0>}]
 neck_02 [{shapeType=Capsule, length=1, radius=5.1544, shapeName=neck_01_capsule, rotation=<rot -90 0 0>}]
}
constraints [
 {bone1Name=dyn_hood_1_l, bone2Name=base_hood_1_l, swing1LimitDegrees=8, swing2LimitDegrees=30, twistLimitDegrees=0}
 {bone1Name=dyn_hood_2_l, bone2Name=base_hood_2_l, swing1LimitDegrees=8, swing2LimitDegrees=30, twistLimitDegrees=0}
 {bone1Name=dyn_hood_c, bone2Name=base_hood_c, swing1LimitDegrees=5, swing2LimitDegrees=30, twistLimitDegrees=0}
 {
  bone1Name=neck_02
  bone2Name=neck_01
  swing1LimitDegrees=45
  swing1Motion=Free
  swing2LimitDegrees=45
  swing2Motion=Free
  twistLimitDegrees=45
  twistMotion=Free
 }
 {
  bone1Name=head
  bone2Name=neck_02
  swing1LimitDegrees=45
  swing1Motion=Free
  swing2LimitDegrees=45
  swing2Motion=Free
  twistLimitDegrees=45
  twistMotion=Free
 }
 {bone1Name=dyn_hood_3_r, bone2Name=base_hood_3_r, swing1LimitDegrees=8, swing2LimitDegrees=30, twistLimitDegrees=0}
 {bone1Name=dyn_hood_1_r, bone2Name=base_hood_1_r, swing1LimitDegrees=8, swing2LimitDegrees=30, twistLimitDegrees=0}
 {bone1Name=dyn_hood_2_r, bone2Name=base_hood_2_r, swing1LimitDegrees=8, swing2LimitDegrees=30, twistLimitDegrees=0}
 {bone1Name=dyn_hood_3_l, bone2Name=base_hood_3_l, swing1LimitDegrees=8, swing2LimitDegrees=30, twistLimitDegrees=0}
]
