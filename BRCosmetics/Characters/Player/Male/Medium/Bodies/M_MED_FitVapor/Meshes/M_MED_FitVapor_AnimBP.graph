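# Post-process-style AnimBP for the M_MED_FitVapor cosmetic body, adding secondary motion to its collar, hood and tailcoat.
# Copies the attached parent's pose into a Control Rig, then chains cached poses: hood, tailcoat and cape-attach offsets, simulation, tailcoat fix.
# AnimDynamics drives the inner collar; two RigidBody nodes use separate physics assets, the hood's with upward gravity.
# EventGraph computes gravity override, skydive/hang-glider hood weight and aim-up crouch correction.
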AnimBlueprint /BRCosmetics/Characters/Player/Male/Medium/Bodies/M_MED_FitVapor/Meshes/M_MED_FitVapor_AnimBP
target_skeleton=&/M_MED_FitVapor_Skeleton
graphs [
 AnimGraph {nodes [
   n1 = "Transform (Modify) Bone - Bone: base_hood_3_l"!(ComponentPose, Alpha=1, Translation="0,0,-1", Rotation="-20,0,0") -> (Pose) {
    rotationMode=BMM_Additive
    rotationSpace=BCS_BoneSpace
    translationMode=BMM_Additive
    translationSpace=BCS_BoneSpace
    alphaBoolBlend {blendOption=Linear}
    alphaScaleBiasClamp {bInterpResult=true, interpSpeedDecreasing=2, interpSpeedIncreasing=2}
    boneToModify {boneName=base_hood_3_l}
    rotation {pitch=-20}
    translation {z=-1}
   }
   n2 = "Transform (Modify) Bone - Bone: base_hood_2_l"!(ComponentPose=@n1, Alpha=1, Translation="0,0,-4", Rotation="-20,0,0") -> (Pose) {
    rotationMode=BMM_Additive
    rotationSpace=BCS_BoneSpace
    translationMode=BMM_Additive
    translationSpace=BCS_BoneSpace
    alphaBoolBlend {blendOption=Linear}
    alphaScaleBiasClamp {bInterpResult=true, interpSpeedDecreasing=2, interpSpeedIncreasing=2}
    boneToModify {boneName=base_hood_2_l}
    rotation {pitch=-20}
    translation {z=-4}
   }
   n3 = "Transform (Modify) Bone - Bone: base_hood_3_r"!(ComponentPose=@n2, Alpha=1, Translation="0,0,-1", Rotation="0,0,0") -> (Pose) {
    rotationMode=BMM_Additive
    rotationSpace=BCS_BoneSpace
    translationMode=BMM_Additive
    translationSpace=BCS_BoneSpace
    alphaBoolBlend {blendOption=Linear}
    alphaScaleBiasClamp {bInterpResult=true, interpSpeedDecreasing=2, interpSpeedIncreasing=2}
    boneToModify {boneName=base_hood_3_r}
    translation {z=-1}
   }
   n4 = "Transform (Modify) Bone - Bone: base_hood_2_r"!(ComponentPose=@n3, Alpha=1, Translation="0,0,-4", Rotation="-20,0,0") -> (Pose) {
    rotationMode=BMM_Additive
    rotationSpace=BCS_BoneSpace
    translationMode=BMM_Additive
    translationSpace=BCS_BoneSpace
    alphaBoolBlend {blendOption=Linear}
    alphaScaleBiasClamp {bInterpResult=true, interpSpeedDecreasing=2, interpSpeedIncreasing=2}
    boneToModify {boneName=base_hood_2_r}
    rotation {pitch=-20}
    translation {z=-4}
   }
   "Transform (Modify) Bone - Bone: base_hood_c"!(ComponentPose=@n4, Alpha=1, Translation="0,0,0", Rotation="0,0,0") -> (Pose) {
    rotationMode=BMM_Additive
    rotationSpace=BCS_BoneSpace
    translationMode=BMM_Additive
    translationSpace=BCS_BoneSpace
    alphaBoolBlend {blendOption=Linear}
    alphaScaleBiasClamp {bInterpResult=true, interpSpeedDecreasing=2, interpSpeedIncreasing=2}
    boneToModify {boneName=base_hood_c}
   }
   Output Pose(Result=Use cached pose 'postSimFixes'()) {layerGroup=None}
   n5 = Control Rig(Source=Copy Pose From Mesh() {bUseAttachedParent=true}) {
    alphaBoolBlend {blendOption=Linear}
    controlRigAssetReference {blueprintRigClass=&/M_MED_FitVapor_CtrlRig.M_MED_FitVapor_CtrlRig_C}
   }
   Save cached pose 'controlRigCP'(Pose=@n5) {cacheName=controlRigCP}
   n6 = Anim Dynamics - Bone: dyn_internalNeckCollar_1_r(ComponentPose=Local To Component(LocalPose=Use cached pose 'generalAdjuistementsCP'()), Alpha=1, ExternalForce="0,0,0") {
    angularDampingOverride=1
    angularSpringConstant=5000
    bAngularSpring=true
    bOverrideAngularDamping=true
    bUseGravityOverride=true
    gravityScale=0.35
    simulationSpace=RootRelative
    alphaBoolBlend {blendOption=Linear}
    boundBone {boneName=dyn_internalNeckCollar_1_r}
    componentAppliedLinearAccClamp=<3000 3000 3000>
    componentLinearAccScale=<0.2 0.2 0.2>
    componentLinearVelScale=<0.2 0.2 0.2>
    gravityOverride {z=-980}
    physicsBodyDefinitions [{
      boundBone {boneName=dyn_internalNeckCollar_1_r}
      boxExtents=<20 30 30>
      constraintSetup {angularLimitsMax=<0 7 15>, angularLimitsMin=<0 -7 -15>, angularTarget=<1 0 0>}
      localJointOffset=<5 0 0>
     }]
   }
   n7 = Anim Dynamics - Bone: dyn_internalNeckCollar_1_c(ComponentPose=@n6, Alpha=1, ExternalForce="0,0,0") {
    angularDampingOverride=1
    angularSpringConstant=5000
    bAngularSpring=true
    bOverrideAngularDamping=true
    bUseGravityOverride=true
    gravityScale=0.35
    simulationSpace=RootRelative
    alphaBoolBlend {blendOption=Linear}
    boundBone {boneName=dyn_internalNeckCollar_1_c}
    componentAppliedLinearAccClamp=<3000 3000 3000>
    componentLinearAccScale=<0.2 0.2 0.2>
    componentLinearVelScale=<0.2 0.2 0.2>
    gravityOverride {z=-980}
    physicsBodyDefinitions [{
      boundBone {boneName=dyn_internalNeckCollar_1_c}
      boxExtents=<20 30 30>
      constraintSetup {angularLimitsMax=<0 7 15>, angularLimitsMin=<0 -7 -15>, angularTarget=<1 0 0>}
      localJointOffset=<5 0 0>
     }]
   }
   n8 = Anim Dynamics - Bone: dyn_internalNeckCollar_1_l(ComponentPose=@n7, Alpha=1, ExternalForce="0,0,0") {
    angularDampingOverride=1
    angularSpringConstant=5000
    bAngularSpring=true
    bOverrideAngularDamping=true
    bUseGravityOverride=true
    gravityScale=0.35
    simulationSpace=RootRelative
    alphaBoolBlend {blendOption=Linear}
    boundBone {boneName=dyn_internalNeckCollar_1_l}
    componentAppliedLinearAccClamp=<3000 3000 3000>
    componentLinearAccScale=<0.2 0.2 0.2>
    componentLinearVelScale=<0.2 0.2 0.2>
    gravityOverride {z=-980}
    physicsBodyDefinitions [{
      boundBone {boneName=dyn_internalNeckCollar_1_l}
      boxExtents=<20 30 30>
      constraintSetup {angularLimitsMax=<0 7 15>, angularLimitsMin=<0 -7 -15>, angularTarget=<1 0 0>}
      localJointOffset=<5 0 0>
     }]
   }
   n9 = RigidBody!(ComponentPose=@n8, Alpha, OverrideWorldGravity:Vector="0,0,0", ExternalForce) -> (Pose) {
    bOverrideWorldGravity=true
    overridePhysicsAsset=&/M_MED_FitVapor_Physics
    simulationSpace=BaseBoneSpace
    alphaBoolBlend {blendOption=Linear}
    alphaScaleBiasClamp {interpSpeedDecreasing=0.1, interpSpeedIncreasing=0.1}
    baseBoneRef {boneName=pelvis}
    componentAppliedLinearAccClamp=<3000 3000 3000>
    componentLinearAccScale=<0.3 0.4 0.3>
    componentLinearVelScale=<0.3 0.4 0.3>
    overrideWorldGravity {z=-980}
    simSpaceSettings {
     maxAngularAcceleration=600
     maxAngularVelocity=1000
     maxLinearAcceleration=700
     maxLinearVelocity=1000
     worldAlpha=0.4
     externalLinearDragV=<0.2 0.2>
    }
   }
   n10 = RigidBody(ComponentPose=@n9, OverrideWorldGravity="0,0,980") {
    bOverrideWorldGravity=true
    overridePhysicsAsset=&/M_MED_FitVapor_Physics_hood
    simulationSpace=BaseBoneSpace
    alphaBoolBlend {blendOption=Linear}
    alphaScaleBiasClamp {interpSpeedDecreasing=0.1, interpSpeedIncreasing=0.1}
    baseBoneRef {boneName=spine_05}
    componentAppliedLinearAccClamp=<3000 3000 3000>
    componentLinearAccScale=<0.01 0.01 0.01>
    componentLinearVelScale=<0.01 0.01 0.01>
    overrideWorldGravity {z=980}
    simSpaceSettings {maxAngularAcceleration=600, maxAngularVelocity=1000, maxLinearAcceleration=700, maxLinearVelocity=1000, externalLinearDragV=<0.2 0.2>}
   }
   Save cached pose 'dynamicsCP'(Pose=Component To Local(ComponentPose=@n10)) {cacheName=dynamicsCP}
   n11 = "Transform (Modify) Bone - Bone: base_tailCoat_back_c_1"!(
    ComponentPose=Local To Component(LocalPose=Use cached pose 'controlRigCP'())
    bAlphaBoolEnabled:bool=True
    Translation="-2,0,2"
    Rotation="40,0,0"
   ) -> (Pose) {
    alphaInputType=Bool
    rotationMode=BMM_Additive
    rotationSpace=BCS_BoneSpace
    translationMode=BMM_Additive
    translationSpace=BCS_BoneSpace
    alphaBoolBlend {blendOption=Sinusoidal, blendOutTime=0.2}
    boneToModify {boneName=base_tailCoat_back_c_1}
    rotation {pitch=40}
    translation {x=-2, z=2}
   }
   n12 = "Transform (Modify) Bone - Bone: base_hood_2_r"(ComponentPose=@n11, Alpha=1, Translation="0,-0.5,-0.5", Rotation="-40,0,0") {
    rotationMode=BMM_Additive
    rotationSpace=BCS_BoneSpace
    translationMode=BMM_Additive
    translationSpace=BCS_BoneSpace
    alphaBoolBlend {blendOption=Linear}
    alphaScaleBiasClamp {bInterpResult=true, interpSpeedDecreasing=2, interpSpeedIncreasing=1}
    boneToModify {boneName=base_hood_2_r}
    rotation {pitch=-40}
    translation {y=-0.5, z=-0.5}
   }
   n13 = "Transform (Modify) Bone - Bone: base_hood_3_l"(ComponentPose=@n12, Alpha=1, Translation="0,0,0", Rotation="-20,0,0") {
    rotationMode=BMM_Additive
    rotationSpace=BCS_BoneSpace
    translationMode=BMM_Additive
    translationSpace=BCS_BoneSpace
    alphaBoolBlend {blendOption=Linear}
    alphaScaleBiasClamp {bInterpResult=true, interpSpeedDecreasing=2, interpSpeedIncreasing=2}
    boneToModify {boneName=base_hood_3_l}
    rotation {pitch=-20}
   }
   n14 = "Transform (Modify) Bone - Bone: base_hood_2_l"(ComponentPose=@n13, Alpha=1, Translation="0,0,-4", Rotation="-20,0,0") {
    rotationMode=BMM_Additive
    rotationSpace=BCS_BoneSpace
    translationMode=BMM_Additive
    translationSpace=BCS_BoneSpace
    alphaBoolBlend {blendOption=Linear}
    alphaScaleBiasClamp {bInterpResult=true, interpSpeedDecreasing=2, interpSpeedIncreasing=2}
    boneToModify {boneName=base_hood_2_l}
    rotation {pitch=-20}
    translation {z=-4}
   }
   n15 = "Transform (Modify) Bone - Bone: base_hood_3_r"(ComponentPose=@n14, Alpha=1, Translation="0,0,0", Rotation="-20,0,0") {
    rotationMode=BMM_Additive
    rotationSpace=BCS_BoneSpace
    translationMode=BMM_Additive
    translationSpace=BCS_BoneSpace
    alphaBoolBlend {blendOption=Linear}
    alphaScaleBiasClamp {bInterpResult=true, interpSpeedDecreasing=2, interpSpeedIncreasing=2}
    boneToModify {boneName=base_hood_3_r}
    rotation {pitch=-20}
   }
   n16 = "Transform (Modify) Bone - Bone: base_hood_2_r"(ComponentPose=@n15, Alpha=1, Translation="0,0,0", Rotation="-20,0,0") {
    rotationMode=BMM_Additive
    rotationSpace=BCS_BoneSpace
    translationMode=BMM_Additive
    translationSpace=BCS_BoneSpace
    alphaBoolBlend {blendOption=Linear}
    alphaScaleBiasClamp {bInterpResult=true, interpSpeedDecreasing=2, interpSpeedIncreasing=2}
    boneToModify {boneName=base_hood_2_r}
    rotation {pitch=-20}
   }
   n17 = "Transform (Modify) Bone - Bone: base_hood_c"(ComponentPose=@n16, Alpha=1, Translation="0,0,0", Rotation="10,0,0") {
    rotationMode=BMM_Additive
    rotationSpace=BCS_BoneSpace
    translationMode=BMM_Additive
    translationSpace=BCS_BoneSpace
    alphaBoolBlend {blendOption=Linear}
    alphaScaleBiasClamp {bInterpResult=true, interpSpeedDecreasing=2, interpSpeedIncreasing=2}
    boneToModify {boneName=base_hood_c}
    rotation {pitch=10}
   }
   Save cached pose 'preSimTweaks'(Pose=Component To Local(ComponentPose=@n17)) {cacheName=preSimTweaks}
   n18 = "Transform (Modify) Bone - Bone: attach_cape"!(
    ComponentPose=Local To Component(LocalPose=Use cached pose 'preSimTweaks'())
    bAlphaBoolEnabled:bool=True
    Translation="0,-1,0"
    Rotation="0,0,16"
   ) -> (Pose) {
    alphaInputType=Bool
    rotationMode=BMM_Additive
    rotationSpace=BCS_BoneSpace
    translationMode=BMM_Additive
    translationSpace=BCS_BoneSpace
    alphaBoolBlend {blendInTime=0.5, blendOption=CircularOut, blendOutTime=0.5}
    boneToModify {boneName=attach_cape}
    rotation {roll=16}
    translation {y=-1}
   }
   n19 = "Transform (Modify) Bone - Bone: attach_cape"(ComponentPose=@n18, bAlphaBoolEnabled=True, Translation="0,-2,0", Rotation="0,0,20") {
    alphaInputType=Bool
    rotationMode=BMM_Additive
    rotationSpace=BCS_BoneSpace
    translationMode=BMM_Additive
    translationSpace=BCS_BoneSpace
    alphaBoolBlend {blendInTime=0.5, blendOption=CircularOut, blendOutTime=0.5}
    boneToModify {boneName=attach_cape}
    rotation {roll=20}
    translation {y=-2}
   }
   n20 = "Transform (Modify) Bone - Bone: attach_cape"(ComponentPose=@n19, bAlphaBoolEnabled=True, Translation="0,0,0", Rotation="0,0,10") {
    alphaInputType=Bool
    rotationMode=BMM_Additive
    rotationSpace=BCS_BoneSpace
    translationMode=BMM_Additive
    translationSpace=BCS_BoneSpace
    alphaBoolBlend {blendInTime=0.5, blendOption=Cubic, blendOutTime=0.5}
    boneToModify {boneName=attach_cape}
    rotation {roll=10}
   }
   Save cached pose 'generalAdjuistementsCP'(Pose=Component To Local(ComponentPose=@n20)) {cacheName=generalAdjuistementsCP}
   n21 = "Transform (Modify) Bone - Bone: dyn_tailCoat_back_c_2"!(
    ComponentPose=Local To Component(LocalPose=Use cached pose 'dynamicsCP'())
    bAlphaBoolEnabled:bool=True
    Translation="0,0,0"
    Rotation="60,20,30"
   ) -> (Pose) {
    alphaInputType=Bool
    rotationMode=BMM_Additive
    rotationSpace=BCS_BoneSpace
    translationMode=BMM_Additive
    translationSpace=BCS_BoneSpace
    alphaBoolBlend {blendInTime=0.2, blendOption=Cubic, blendOutTime=0.2}
    boneToModify {boneName=dyn_tailCoat_back_c_2}
    rotation=<rot 60 20 30>
   }
   Save cached pose 'postSimFixes'(Pose=Component To Local(ComponentPose=@n21)) {cacheName=postSimFixes}
  ]}
 EventGraph {nodes [
   Event BlueprintUpdateAnimation()
   > n1 = Sequence!(execute) -> (then_0, then_1, then_2:Exec)
   > n2 = "Gravity Override State Function (Parameter Asset) (Nativized)"(
    CharacterPartAnimInstance=Self-Reference()
    GameTimeInSeconds=GetGameTimeInSeconds()
    Parameters=&/M_MED_FitVapor_Parameters
   )
   > Set GravityOverride(GravityOverride=@n2)
   n3 = Branch(execute=@n1.then_1, Condition=Get bIsSkydiveFloating())
   > Set weightAdjustHood(weightAdjustHood:double=1) -> (Output_Get:double)
   n4 = Branch(execute=@n3.else, Condition=Get isUsingHangGlidingInParachute() -> (isUsingHangGlidingInParachute:bool))
   > Set weightAdjustHood(weightAdjustHood=0.5)
   Set weightAdjustHood(execute=@n4.else, weightAdjustHood=0)
   n5 = AND Boolean!(A=NOT Boolean(A=Get bIsMoving2D()), B=Get bIsCrouching(), C:bool=Get bIsTargeting()) -> (ReturnValue)
   n6 = Get AimPitch()
   Set toogleFixASDUpCrouching(
    execute=@n1.then_2
    toogleFixASDUpCrouching:double=float * float(A=Remap Float(Input=@n6, OldRange=(X=0,Y=90), NewRange=(X=0,Y=1), Clamp=true) -> (Output=~), B="To Float (Boolean)"(InBool=@n5)) {is=title K2Node_PromotableOperator}
   ) -> (Output_Get:double)
   Event OnParachuteOpened()
   > Switch on EGliderType(Selection=Get PlayerGliderType() -> (PlayerGliderType=HangGlider))
   > Set isUsingHangGlidingInParachute(isUsingHangGlidingInParachute:bool=true) -> (Output_Get:bool)
   Event OnParachuteClosed()
   > Set isUsingHangGlidingInParachute(isUsingHangGlidingInParachute=false)
  ]}
]
variables [
 GravityOverride {type=Vector}
 isUsingHangGlidingInParachute {type=boolean}
 toogleFixASDUpCrouching {type=number}
 weightAdjustHood {type=number}
]
