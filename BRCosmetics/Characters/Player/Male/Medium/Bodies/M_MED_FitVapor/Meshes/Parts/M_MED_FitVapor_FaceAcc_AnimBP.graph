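# Anim Blueprint for the FitVapor character's FaceAcc part, adding secondary motion to its hair strands.
# It copies the attached parent's pose, simulates two main strand bones with sprung AnimDynamics, rivets ten base bones to weighted blends of those and the head, then simulates ten smaller strands.
# The update event stores a GravityOverride from a parameter asset, but the dynamics nodes use literal gravity, inverted on strand 6.
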AnimBlueprint /BRCosmetics/Characters/Player/Male/Medium/Bodies/M_MED_FitVapor/Meshes/Parts/M_MED_FitVapor_FaceAcc_AnimBP
target_skeleton=&/M_MED_FitVapor_FaceAcc_Skeleton
graphs [
 AnimGraph {nodes [
   Output Pose(Result=Use cached pose 'secondaryDynamicsCP'()) {layerGroup=None}
   Save cached pose 'InputCP'(Pose=Copy Pose From Mesh() {bUseAttachedParent=true}) {cacheName=InputCP}
   n1 = Anim Dynamics - Bone: dyn_hairStrandMain_1!(ComponentPose=Local To Component(LocalPose=Use cached pose 'InputCP'()), Alpha=0.8, GravityOverride:Vector="0,0,-980", ExternalForce="0,0,0") -> (Pose) {
    angularDampingOverride=1
    angularSpringConstant=2000
    bAngularSpring=true
    bOverrideAngularDamping=true
    bUseGravityOverride=true
    gravityScale=0.35
    simulationSpace=RootRelative
    alphaBoolBlend {blendOption=Linear}
    boundBone {boneName=dyn_hairStrandMain_1}
    componentAppliedLinearAccClamp=<3000 3000 3000>
    componentLinearAccScale=<1 1 1>
    componentLinearVelScale=<1 1 1>
    gravityOverride {z=-980}
    physicsBodyDefinitions [{
      boundBone {boneName=dyn_hairStrandMain_1}
      boxExtents=<15 25 25>
      constraintSetup {angularLimitsMax=<0 15 20>, angularLimitsMin=<0 -5 -20>, angularTarget=<1 0 0>}
      localJointOffset=<4 0 0>
     }]
   }
   n2 = Anim Dynamics - Bone: dyn_hairStrandMain_2!(ComponentPose=@n1, Alpha=0.8, GravityOverride:Vector="0,0,-980", ExternalForce="0,0,0") -> (Pose) {
    angularDampingOverride=1
    angularSpringConstant=2000
    bAngularSpring=true
    bOverrideAngularDamping=true
    bUseGravityOverride=true
    gravityScale=0.35
    simulationSpace=RootRelative
    alphaBoolBlend {blendOption=Linear}
    boundBone {boneName=dyn_hairStrandMain_2}
    componentAppliedLinearAccClamp=<3000 3000 3000>
    componentLinearAccScale=<1 1 1>
    componentLinearVelScale=<1 1 1>
    gravityOverride {z=-980}
    physicsBodyDefinitions [{
      boundBone {boneName=dyn_hairStrandMain_2}
      boxExtents=<15 25 25>
      constraintSetup {angularLimitsMax=<0 15 20>, angularLimitsMin=<0 -7 -20>, angularTarget=<1 0 0>}
      localJointOffset=<4 0 0>
     }]
   }
   Save cached pose 'dynamicsCP'(Pose=Component To Local(ComponentPose=@n2)) {cacheName=dynamicsCP}
   n3 = Anim Dynamics - Bone: dyn_hairStrand_1!(
    ComponentPose=Local To Component(LocalPose=Use cached pose 'rivetsCP'())
    Alpha=0.8
    GravityOverride:Vector="0,0,-980"
    ExternalForce="0,0,0"
   ) -> (Pose) {
    angularDampingOverride=1
    angularSpringConstant=3000
    bAngularSpring=true
    bOverrideAngularDamping=true
    bUseGravityOverride=true
    gravityScale=0.35
    simulationSpace=RootRelative
    alphaBoolBlend {blendOption=Linear}
    boundBone {boneName=dyn_hairStrand_1}
    componentAppliedLinearAccClamp=<3000 3000 3000>
    componentLinearAccScale=<1 1 1>
    componentLinearVelScale=<1 1 1>
    gravityOverride {z=-980}
    physicsBodyDefinitions [{
      boundBone {boneName=dyn_hairStrand_1}
      boxExtents=<15 25 25>
      constraintSetup {angularLimitsMax=<0 10 10>, angularLimitsMin=<0 -10 -10>, angularTarget=<1 0 0>}
      localJointOffset=<4 0 0>
     }]
   }
   n4 = Anim Dynamics - Bone: dyn_hairStrand_2!(ComponentPose=@n3, Alpha=0.8, GravityOverride:Vector="0,0,-980", ExternalForce="0,0,0") -> (Pose) {
    angularDampingOverride=1
    angularSpringConstant=3000
    bAngularSpring=true
    bOverrideAngularDamping=true
    bUseGravityOverride=true
    gravityScale=0.35
    simulationSpace=RootRelative
    alphaBoolBlend {blendOption=Linear}
    boundBone {boneName=dyn_hairStrand_2}
    componentAppliedLinearAccClamp=<3000 3000 3000>
    componentLinearAccScale=<1 1 1>
    componentLinearVelScale=<1 1 1>
    gravityOverride {z=-980}
    physicsBodyDefinitions [{
      boundBone {boneName=dyn_hairStrand_2}
      boxExtents=<15 25 25>
      constraintSetup {angularLimitsMax=<0 10 10>, angularLimitsMin=<0 -10 -10>, angularTarget=<1 0 0>}
      localJointOffset=<4 0 0>
     }]
   }
   n5 = Anim Dynamics - Bone: dyn_hairStrand_3!(ComponentPose=@n4, Alpha=0.8, GravityOverride:Vector="0,0,-980", ExternalForce="0,0,0") -> (Pose) {
    angularDampingOverride=1
    angularSpringConstant=3000
    bAngularSpring=true
    bOverrideAngularDamping=true
    bUseGravityOverride=true
    gravityScale=0.35
    simulationSpace=RootRelative
    alphaBoolBlend {blendOption=Linear}
    boundBone {boneName=dyn_hairStrand_3}
    componentAppliedLinearAccClamp=<3000 3000 3000>
    componentLinearAccScale=<1 1 1>
    componentLinearVelScale=<1 1 1>
    gravityOverride {z=-980}
    physicsBodyDefinitions [{
      boundBone {boneName=dyn_hairStrand_3}
      boxExtents=<15 25 25>
      constraintSetup {angularLimitsMax=<0 10 10>, angularLimitsMin=<0 -10 -10>, angularTarget=<1 0 0>}
      localJointOffset=<4 0 0>
     }]
   }
   n6 = Anim Dynamics - Bone: dyn_hairStrand_4!(ComponentPose=@n5, Alpha=0.8, GravityOverride:Vector="0,0,-980", ExternalForce="0,0,0") -> (Pose) {
    angularDampingOverride=1
    angularSpringConstant=3000
    bAngularSpring=true
    bOverrideAngularDamping=true
    bUseGravityOverride=true
    gravityScale=0.35
    simulationSpace=RootRelative
    alphaBoolBlend {blendOption=Linear}
    boundBone {boneName=dyn_hairStrand_4}
    componentAppliedLinearAccClamp=<3000 3000 3000>
    componentLinearAccScale=<1 1 1>
    componentLinearVelScale=<1 1 1>
    gravityOverride {z=-980}
    physicsBodyDefinitions [{
      boundBone {boneName=dyn_hairStrand_4}
      boxExtents=<15 25 25>
      constraintSetup {angularLimitsMax=<0 10 10>, angularLimitsMin=<0 -10 -10>, angularTarget=<1 0 0>}
      localJointOffset=<4 0 0>
     }]
   }
   n7 = Anim Dynamics - Bone: dyn_hairStrand_5!(ComponentPose=@n6, Alpha=0.8, GravityOverride:Vector="0,0,-980", ExternalForce="0,0,0") -> (Pose) {
    angularDampingOverride=1
    angularSpringConstant=3000
    bAngularSpring=true
    bOverrideAngularDamping=true
    bUseGravityOverride=true
    gravityScale=0.35
    simulationSpace=RootRelative
    alphaBoolBlend {blendOption=Linear}
    boundBone {boneName=dyn_hairStrand_5}
    componentAppliedLinearAccClamp=<3000 3000 3000>
    componentLinearAccScale=<1 1 1>
    componentLinearVelScale=<1 1 1>
    gravityOverride {z=-980}
    physicsBodyDefinitions [{
      boundBone {boneName=dyn_hairStrand_5}
      boxExtents=<15 25 25>
      constraintSetup {angularLimitsMax=<0 10 10>, angularLimitsMin=<0 0 -10>, angularTarget=<1 0 0>}
      localJointOffset=<4 0 0>
     }]
   }
   n8 = Anim Dynamics - Bone: dyn_hairStrand_6!(ComponentPose=@n7, Alpha=0.8, GravityOverride:Vector="0,0,980", ExternalForce="0,0,0") -> (Pose) {
    angularDampingOverride=1
    angularSpringConstant=3000
    bAngularSpring=true
    bOverrideAngularDamping=true
    bUseGravityOverride=true
    gravityScale=0.35
    simulationSpace=RootRelative
    alphaBoolBlend {blendOption=Linear}
    boundBone {boneName=dyn_hairStrand_6}
    componentAppliedLinearAccClamp=<3000 3000 3000>
    componentLinearAccScale=<1 1 1>
    componentLinearVelScale=<1 1 1>
    gravityOverride {z=980}
    physicsBodyDefinitions [{
      boundBone {boneName=dyn_hairStrand_6}
      boxExtents=<15 25 25>
      constraintSetup {angularLimitsMax=<0 10 10>, angularLimitsMin=<0 -10 -10>, angularTarget=<1 0 0>}
      localJointOffset=<4 0 0>
     }]
   }
   n9 = Anim Dynamics - Bone: dyn_hairStrand_7!(ComponentPose=@n8, Alpha=0.8, GravityOverride:Vector="0,0,-980", ExternalForce="0,0,0") -> (Pose) {
    angularDampingOverride=1
    angularSpringConstant=3000
    bAngularSpring=true
    bOverrideAngularDamping=true
    bUseGravityOverride=true
    gravityScale=0.35
    simulationSpace=RootRelative
    alphaBoolBlend {blendOption=Linear}
    boundBone {boneName=dyn_hairStrand_7}
    componentAppliedLinearAccClamp=<3000 3000 3000>
    componentLinearAccScale=<1 1 1>
    componentLinearVelScale=<1 1 1>
    gravityOverride {z=-980}
    physicsBodyDefinitions [{
      boundBone {boneName=dyn_hairStrand_7}
      boxExtents=<15 25 25>
      constraintSetup {angularLimitsMax=<0 10 10>, angularLimitsMin=<0 -10 -10>, angularTarget=<1 0 0>}
      localJointOffset=<4 0 0>
     }]
   }
   n10 = Anim Dynamics - Bone: dyn_hairStrand_8!(ComponentPose=@n9, Alpha=0.8, GravityOverride:Vector="0,0,-980", ExternalForce="0,0,0") -> (Pose) {
    angularDampingOverride=1
    angularSpringConstant=3000
    bAngularSpring=true
    bOverrideAngularDamping=true
    bUseGravityOverride=true
    gravityScale=0.35
    simulationSpace=RootRelative
    alphaBoolBlend {blendOption=Linear}
    boundBone {boneName=dyn_hairStrand_8}
    componentAppliedLinearAccClamp=<3000 3000 3000>
    componentLinearAccScale=<1 1 1>
    componentLinearVelScale=<1 1 1>
    gravityOverride {z=-980}
    physicsBodyDefinitions [{
      boundBone {boneName=dyn_hairStrand_8}
      boxExtents=<15 25 25>
      constraintSetup {angularLimitsMax=<0 10 10>, angularLimitsMin=<0 -10 -10>, angularTarget=<1 0 0>}
      localJointOffset=<4 0 0>
     }]
   }
   n11 = Anim Dynamics - Bone: dyn_hairStrand_9!(ComponentPose=@n10, Alpha=0.8, GravityOverride:Vector="0,0,-980", ExternalForce="0,0,0") -> (Pose) {
    angularDampingOverride=1
    angularSpringConstant=3000
    bAngularSpring=true
    bOverrideAngularDamping=true
    bUseGravityOverride=true
    gravityScale=0.35
    simulationSpace=RootRelative
    alphaBoolBlend {blendOption=Linear}
    boundBone {boneName=dyn_hairStrand_9}
    componentAppliedLinearAccClamp=<3000 3000 3000>
    componentLinearAccScale=<1 1 1>
    componentLinearVelScale=<1 1 1>
    gravityOverride {z=-980}
    physicsBodyDefinitions [{
      boundBone {boneName=dyn_hairStrand_9}
      boxExtents=<15 25 25>
      constraintSetup {angularLimitsMax=<0 10 10>, angularLimitsMin=<0 -10 -10>, angularTarget=<1 0 0>}
      localJointOffset=<4 0 0>
     }]
   }
   n12 = Anim Dynamics - Bone: dyn_hairStrand_10!(ComponentPose=@n11, Alpha=0.8, GravityOverride:Vector="0,0,-980", ExternalForce="0,0,0") -> (Pose) {
    angularDampingOverride=1
    angularSpringConstant=3000
    bAngularSpring=true
    bOverrideAngularDamping=true
    bUseGravityOverride=true
    gravityScale=0.35
    simulationSpace=RootRelative
    alphaBoolBlend {blendOption=Linear}
    boundBone {boneName=dyn_hairStrand_10}
    componentAppliedLinearAccClamp=<3000 3000 3000>
    componentLinearAccScale=<1 1 1>
    componentLinearVelScale=<1 1 1>
    gravityOverride {z=-980}
    physicsBodyDefinitions [{
      boundBone {boneName=dyn_hairStrand_10}
      boxExtents=<15 25 25>
      constraintSetup {angularLimitsMax=<0 10 10>, angularLimitsMin=<0 -10 -10>, angularTarget=<1 0 0>}
      localJointOffset=<4 0 0>
     }]
   }
   Save cached pose 'secondaryDynamicsCP'(Pose=Component To Local(ComponentPose=@n12)) {cacheName=secondaryDynamicsCP}
   n13 = "Constraint - base_hairStrand_1 (PP)"!(
    ComponentPose=Local To Component(LocalPose=Use cached pose 'dynamicsCP'())
    Alpha=1
    ConstraintWeights_0=0.9926
    ConstraintWeights_1:float=0.0074
   ) -> (Pose) {
    alphaBoolBlend {blendOption=Linear}
    boneToModify {boneName=base_hairStrand_1}
    constraintSetup [{targetBone {boneName=head}}, {targetBone {boneName=dyn_hairStrandMain_1}}]
    constraintWeights [1,1]
   }
   n14 = "Constraint - base_hairStrand_2 (PPP)"!(ComponentPose=@n13, Alpha=1, ConstraintWeights_0=0.6863, ConstraintWeights_1:float=0.3021, ConstraintWeights_2:float=0.0116) -> (Pose) {
    alphaBoolBlend {blendOption=Linear}
    boneToModify {boneName=base_hairStrand_2}
    constraintSetup [{targetBone {boneName=head}}, {targetBone {boneName=dyn_hairStrandMain_1}}, {targetBone {boneName=dyn_hairStrandMain_2}}]
    constraintWeights [1,1,1]
   }
   n15 = "Constraint - base_hairStrand_3 (PP)"!(ComponentPose=@n14, Alpha=1, ConstraintWeights_0=0.9708, ConstraintWeights_1:float=0.0292) -> (Pose) {
    alphaBoolBlend {blendOption=Linear}
    boneToModify {boneName=base_hairStrand_3}
    constraintSetup [{targetBone {boneName=head}}, {targetBone {boneName=dyn_hairStrandMain_1}}]
    constraintWeights [1,1]
   }
   n16 = "Constraint - base_hairStrand_4 (PP)"!(ComponentPose=@n15, Alpha=1, ConstraintWeights_0=0.996, ConstraintWeights_1:float=0.004) -> (Pose) {
    alphaBoolBlend {blendOption=Linear}
    boneToModify {boneName=base_hairStrand_4}
    constraintSetup [{targetBone {boneName=head}}, {targetBone {boneName=dyn_hairStrandMain_1}}]
    constraintWeights [1,1]
   }
   n17 = "Constraint - base_hairStrand_5 (PPP)"!(ComponentPose=@n16, Alpha=1, ConstraintWeights_0=0.6655, ConstraintWeights_1:float=0.2553, ConstraintWeights_2:float=0.0792) -> (Pose) {
    alphaBoolBlend {blendOption=Linear}
    boneToModify {boneName=base_hairStrand_5}
    constraintSetup [{targetBone {boneName=dyn_hairStrandMain_2}}, {targetBone {boneName=dyn_hairStrandMain_1}}, {targetBone {boneName=head}}]
    constraintWeights [1,1,1]
   }
   n18 = "Constraint - base_hairStrand_6 (PPP)"!(ComponentPose=@n17, Alpha=1, ConstraintWeights_0=0.4025, ConstraintWeights_1:float=0.3623, ConstraintWeights_2:float=0.2352) -> (Pose) {
    alphaBoolBlend {blendOption=Linear}
    boneToModify {boneName=base_hairStrand_6}
    constraintSetup [{targetBone {boneName=dyn_hairStrandMain_1}}, {targetBone {boneName=dyn_hairStrandMain_2}}, {targetBone {boneName=head}}]
    constraintWeights [1,1,1]
   }
   n19 = "Constraint - base_hairStrand_7 (PPP)"!(ComponentPose=@n18, Alpha=1, ConstraintWeights_0=0.4417, ConstraintWeights_1:float=0.4309, ConstraintWeights_2:float=0.1275) -> (Pose) {
    alphaBoolBlend {blendOption=Linear}
    boneToModify {boneName=base_hairStrand_7}
    constraintSetup [{targetBone {boneName=dyn_hairStrandMain_1}}, {targetBone {boneName=dyn_hairStrandMain_2}}, {targetBone {boneName=head}}]
    constraintWeights [1,1,1]
   }
   n20 = "Constraint - base_hairStrand_8 (PPP)"!(ComponentPose=@n19, Alpha=1, ConstraintWeights_0=0.7358, ConstraintWeights_1:float=0.2274, ConstraintWeights_2:float=0.0368) -> (Pose) {
    alphaBoolBlend {blendOption=Linear}
    boneToModify {boneName=base_hairStrand_8}
    constraintSetup [{targetBone {boneName=dyn_hairStrandMain_2}}, {targetBone {boneName=dyn_hairStrandMain_1}}, {targetBone {boneName=head}}]
    constraintWeights [1,1,1]
   }
   n21 = "Constraint - base_hairStrand_9 (PPP)"!(ComponentPose=@n20, Alpha=1, ConstraintWeights_0=0.6345, ConstraintWeights_1:float=0.2076, ConstraintWeights_2:float=0.1579) -> (Pose) {
    alphaBoolBlend {blendOption=Linear}
    boneToModify {boneName=base_hairStrand_9}
    constraintSetup [{targetBone {boneName=dyn_hairStrandMain_2}}, {targetBone {boneName=dyn_hairStrandMain_1}}, {targetBone {boneName=head}}]
    constraintWeights [1,1,1]
   }
   n22 = "Constraint - base_hairStrand_10 (PPP)"!(ComponentPose=@n21, Alpha=1, ConstraintWeights_0=0.4793, ConstraintWeights_1:float=0.3738, ConstraintWeights_2:float=0.1469) -> (Pose) {
    alphaBoolBlend {blendOption=Linear}
    boneToModify {boneName=base_hairStrand_10}
    constraintSetup [{targetBone {boneName=dyn_hairStrandMain_1}}, {targetBone {boneName=head}}, {targetBone {boneName=dyn_hairStrandMain_2}}]
    constraintWeights [1,1,1]
   }
   Save cached pose 'rivetsCP'(Pose=Component To Local(ComponentPose=@n22)) {cacheName=rivetsCP}
  ]}
 EventGraph {nodes [
   Event BlueprintUpdateAnimation()
   > n1 = "Gravity Override State Function (Parameter Asset) (Nativized)"(
    CharacterPartAnimInstance=Self-Reference()
    GameTimeInSeconds=GetGameTimeInSeconds()
    Parameters=/BRCosmetics/Characters/Player/Male/Medium/Bodies/M_MED_FitVapor/Meshes/M_MED_FitVapor_Parameters
   )
   > Set GravityOverride(GravityOverride=@n1)
  ]}
]
variables [GravityOverride {type=Vector}]
